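# Control Rig for the Dame Regent female body that adds corrective motion to the skirt and accessory bones.
# After a shared deform-rig function, parent constraints blend chain and dress bones between neighbours.
# Swing-twist readings of thighs and right calf, plus pelvis height, become weights for additive transform correctives, some multiplied for combined poses.
# Crouch, slide and jump inputs drive further fixes, mostly interpolation-smoothed; the floating fix and necklace constraints are unwired.
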ControlRigBlueprint /BRCosmetics/Characters/Player/Female/Medium/Bodies/F_MED_DameRegent/Meshes/F_MED_DameRegent_CtrlRig
graphs [
 Multiply_6_ContainedGraph {nodes [
   n1 = RigVMFunctionEntryNode()
   RigVMFunctionReturnNode(Result=&RigVMFunction_MathDoubleMul(B=@n1.C, A=&RigVMFunction_MathDoubleMul(A=@n1.A, B=@n1.B).Result).Result)
  ]}
 New Function_ContainedGraph {nodes [
   n1 = RigVMFunctionEntryNode()
   RigVMFunction_Sequence = &RigVMFunction_Sequence(ExecuteContext=@n1.ExecuteContext)
   DeltaRotation v4 = RigVMFunctionReferenceNode(
    Rotation Order=ZYX
    child=(Type=Bone,Name="thigh_l")
    debugRotation=false
    outputMinMax=(X=0,Y=1)
    parent=(Type=Null,Name="thigh_l_null")
    premult=1
    rotationMinMax=(X=0,Y=-110)
    triggerAxis=Z
    twistAxis=(X=1,Y=0,Z=0)
    ExecuteContext=@RigVMFunction_Sequence.A
   )
   > RigVMVariableNode(Value=@"DeltaRotation v4".outputWeight) {variable=SPR L Leg Fwd}
   > DeltaRotation v4_1 = RigVMFunctionReferenceNode(
    Rotation Order=ZYX
    child=(Type=Bone,Name="thigh_r")
    debugRotation=false
    outputMinMax=(X=0,Y=1)
    parent=(Type=Null,Name="thigh_r_null")
    premult=1
    rotationMinMax=(X=0,Y=-110)
    triggerAxis=Z
    twistAxis=(X=1,Y=0,Z=0)
   )
   > RigVMVariableNode(Value=@"DeltaRotation v4_1".outputWeight) {variable=SPR R Leg Fwd}
   > DeltaRotation v4_2 = RigVMFunctionReferenceNode(
    Rotation Order=ZYX
    child=(Type=Bone,Name="thigh_l")
    debugRotation=false
    outputMinMax=(X=0,Y=1)
    parent=(Type=Null,Name="thigh_l_null")
    premult=1
    rotationMinMax=(X=0,Y=110)
    triggerAxis=Z
    twistAxis=(X=1,Y=0,Z=0)
   )
   > RigVMVariableNode(Value=@"DeltaRotation v4_2".outputWeight) {variable=SPR L Leg Bkd}
   > DeltaRotation v4_3 = RigVMFunctionReferenceNode(
    Rotation Order=ZYX
    child=(Type=Bone,Name="thigh_r")
    debugRotation=false
    outputMinMax=(X=0,Y=1)
    parent=(Type=Null,Name="thigh_r_null")
    premult=1
    rotationMinMax=(X=0,Y=110)
    triggerAxis=Z
    twistAxis=(X=1,Y=0,Z=0)
   )
   > RigVMVariableNode(Value=@"DeltaRotation v4_3".outputWeight) {variable=SPR R Leg Bkd}
   > DeltaRotation v4_4 = RigVMFunctionReferenceNode(
    Rotation Order=ZYX
    child=(Type=Bone,Name="thigh_l")
    debugRotation=false
    outputMinMax=(X=0,Y=1)
    parent=(Type=Null,Name="thigh_l_null")
    premult=1
    rotationMinMax=(X=0,Y=-90)
    triggerAxis=Y
    twistAxis=(X=1,Y=0,Z=0)
   )
   > RigVMVariableNode(Value=@"DeltaRotation v4_4".outputWeight) {variable=SPR L Leg Otw}
   > DeltaRotation v4_5 = RigVMFunctionReferenceNode(
    Rotation Order=ZYX
    child=(Type=Bone,Name="thigh_r")
    debugRotation=false
    outputMinMax=(X=0,Y=1)
    parent=(Type=Null,Name="thigh_r_null")
    premult=1
    rotationMinMax=(X=0,Y=-90)
    triggerAxis=Y
    twistAxis=(X=1,Y=0,Z=0)
   )
   > RigVMVariableNode(Value=@"DeltaRotation v4_5".outputWeight) {variable=SPR R Leg Otw}
   > DeltaRotation v4_6 = RigVMFunctionReferenceNode(
    Rotation Order=ZYX
    child=(Type=Bone,Name="calf_r")
    debugRotation=false
    outputMinMax=(X=0,Y=1)
    parent=(Type=Null,Name="calf_r_null")
    premult=1
    rotationMinMax=(X=0,Y=120)
    triggerAxis=Z
    twistAxis=(X=1,Y=0,Z=0)
   )
   > RigVMVariableNode(Value=@"DeltaRotation v4_6".outputWeight) {variable=SPR R Calf Bkd}
   > DeltaRotation v4_7 = RigVMFunctionReferenceNode(
    Rotation Order=ZYX
    child=(Type=Bone,Name="calf_r")
    debugRotation=false
    onlyTwist=false
    outputMinMax=(X=0,Y=1)
    parent=(Type=Null,Name="calf_r_null")
    premult=1
    rotationMinMax=(X=0,Y=-30)
    triggerAxis=Y
    twistAxis=(X=1,Y=0,Z=0)
   )
   > RigVMVariableNode(Value=@"DeltaRotation v4_7".outputWeight) {variable=SPR R Calf Twist}
   > DeltaRotation v4_8 = RigVMFunctionReferenceNode(
    Rotation Order=ZYX
    child=(Type=Bone,Name="thigh_r")
    debugRotation=false
    onlyTwist=false
    outputMinMax=(X=0,Y=1)
    parent=(Type=Null,Name="thigh_r_null")
    premult=1
    rotationMinMax=(X=-110,Y=-140)
    triggerAxis=Z
    twistAxis=(X=1,Y=0,Z=0)
   )
   > RigVMVariableNode(Value=@"DeltaRotation v4_8".outputWeight) {variable=SPR R Leg Fwd_Extreme}
   > DeltaRotation v4_9 = RigVMFunctionReferenceNode(
    Rotation Order=ZYX
    child=(Type=Bone,Name="thigh_r")
    debugRotation=false
    onlyTwist=false
    outputMinMax=(X=0,Y=1)
    parent=(Type=Null,Name="thigh_r_null")
    premult=1
    rotationMinMax=(X=0,Y=30)
    triggerAxis=Y
    twistAxis=(X=1,Y=0,Z=0)
   )
   > RigVMVariableNode(Value=@"DeltaRotation v4_9".outputWeight) {variable=SPR R Leg Itw}
   > DeltaRotation v4_10 = RigVMFunctionReferenceNode(
    Rotation Order=ZYX
    child=(Type=Bone,Name="thigh_l")
    debugRotation=false
    onlyTwist=false
    outputMinMax=(X=0,Y=1)
    parent=(Type=Null,Name="thigh_l_null")
    premult=1
    rotationMinMax=(X=0,Y=30)
    triggerAxis=Y
    twistAxis=(X=1,Y=0,Z=0)
   )
   > RigVMVariableNode() {variable=SPR L Leg Itw}
   n2 = &RigVMFunction_AnimEvalRichCurve(
    Curve="(EditorCurveData=(Keys=((InterpMode=RCIM_Cubic,TangentMode=RCTM_Break,TangentWeightMode=RCTWM_WeightedNone,Time=0,Value=0,ArriveTangent=0,ArriveTangentWeight=0,LeaveTangent=0,LeaveTangentWeight=0),(InterpMode=RCIM_Cubic,TangentMode=RCTM_Break,TangentWeight ...(truncated)"
    SourceMinimum=0.5
    TargetMinimum=0
    Value=&RigVMFunction_MathDoubleRemap(
     SourceMaximum=30
     SourceMinimum=60
     TargetMaximum=0
     TargetMinimum=1
     bClamp=False
     Value=&RigUnit_GetRelativeTransformForItem(Child=(Type=Bone,Name="pelvis"), Parent=(Type=Bone,Name="root"), bChildInitial=False, bParentInitial=False).RelativeTransform.Translation.Z
    ).Result
   )
   > RigVMVariableNode(Value=@n2.Result) {variable=PelvisProx}
   RigVMFunctionReturnNode(ExecuteContext=@RigVMFunction_Sequence.B)
  ]}
 New Function_ContainedGraph {nodes [
   n1 = RigVMFunctionEntryNode()
   RigVMFunction_MathBoolToInteger = &RigVMFunction_MathBoolToInteger(Value=@n1.onlyTwist)
   RigVMFunction_MathQuaternionSwingTwist = &RigVMFunction_MathQuaternionSwingTwist(
    TwistAxis=@n1.twistAxis
    Input=&RigUnit_GetRelativeTransformForItem(bChildInitial=False, bParentInitial=False, Child=@n1.child, Parent=@n1.parent).RelativeTransform.Rotation
   )
   n2 = RigVMDispatchNode(
    Values=((),())
    Index=@RigVMFunction_MathBoolToInteger.Result
    Values.0=@RigVMFunction_MathQuaternionSwingTwist.Swing
    Values.1=@RigVMFunction_MathQuaternionSwingTwist.Twist
   ) {template="DISPATCH_RigVMDispatch_SelectInt32(in Index,in Values,out Result)"}
   RigVMFunction_MathQuaternionToEuler = &RigVMFunction_MathQuaternionToEuler(Value=@n2.Result, RotationOrder=@n1.Rotation Order)
   > DISPATCH_RigVMDispatch_Print = RigVMDispatchNode(ScreenColor=(R=1,G=1,B=1,A=1), ScreenDuration=1.5, Enabled=@n1.debugRotation, Value=@RigVMFunction_MathQuaternionToEuler.Result) {template="DISPATCH_RigVMDispatch_Print(in Prefix,in Value,in Enabled,in ScreenDuration,in ScreenColor)"}
   Cast = RigVMDispatchNode(Value=@n1.triggerAxis) {template="DISPATCH_RigVMDispatch_CastEnumToInt(in Value,out Result)"}
   DISPATCH_RigVMDispatch_SelectInt32 = RigVMDispatchNode(
    Values=(0,0,0)
    Index=@Cast.Result
    Values.0=@RigVMFunction_MathQuaternionToEuler.Result.X
    Values.1=@RigVMFunction_MathQuaternionToEuler.Result.Y
    Values.2=@RigVMFunction_MathQuaternionToEuler.Result.Z
   ) {template="DISPATCH_RigVMDispatch_SelectInt32(in Index,in Values,out Result)"}
   Remap = &RigVMFunction_MathDoubleRemap(
    bClamp=False
    Value=@DISPATCH_RigVMDispatch_SelectInt32.Result
    TargetMinimum=@n1.outputMinMax.X
    TargetMaximum=@n1.outputMinMax.Y
    SourceMinimum=@n1.rotationMinMax.X
    SourceMaximum=@n1.rotationMinMax.Y
   )
   Absolute = &RigVMFunction_MathDoubleAbs(Value=@Remap.Result)
   Multiply = &RigVMFunction_MathDoubleMul(
    A=&RigVMFunction_MathDoubleClamp(Value=RigVMDispatchNode(True=@Absolute.Result, Condition=@n1.absoluteValue, False=@Remap.Result) {template="DISPATCH_RigVMDispatch_If(in Condition,in True,in False,out Result)"}.Result).Result
    B=@n1.premult
   )
   > RigVMFunctionReturnNode(outputQuat=@n2.Result, outputWeight=@Multiply.Result)
  ]}
 New Function_ContainedGraph {nodes [
   n1 = RigVMFunctionEntryNode()
   RigVMFunction_ControlFlowBranch = &RigVMFunction_ControlFlowBranch(Condition=@n1.Debug Controllers, ExecuteContext=@n1.ExecuteContext)
   RigVMFunctionReturnNode(ExecuteContext=@RigVMFunction_ControlFlowBranch.Completed)
   RigUnit_GetTransform = &RigUnit_GetTransform(Item=(Type=Control,Name="thigh_l_ctrl"), bInitial=False)
   &RigUnit_SetTransform(
    Item=(Type=Bone,Name="thigh_l")
    Weight=1
    bInitial=False
    Value=@RigUnit_GetTransform.Transform
    ExecutePin=@RigVMFunction_ControlFlowBranch.True
   )
   > &RigUnit_SetTransform(
    Item=(Type=Bone,Name="calf_l")
    Weight=1
    bInitial=False
    Value=&RigUnit_GetTransform(Item=(Type=Control,Name="calf_l_ctrl"), bInitial=False).Transform
   )
   > RigUnit_SetTransform = &RigUnit_SetTransform(
    Item=(Type=Bone,Name="thigh_r")
    Weight=1
    bInitial=False
    Value=&RigUnit_GetTransform(Item=(Type=Control,Name="thigh_r_ctrl"), bInitial=False).Transform
   )
   > &RigUnit_SetTransform(
    Item=(Type=Bone,Name="calf_r")
    Weight=1
    bInitial=False
    Value=&RigUnit_GetTransform(Item=(Type=Control,Name="calf_r_ctrl"), bInitial=False).Transform
   )
  ]}
 New Function_ContainedGraph {nodes [
   n1 = RigVMFunctionEntryNode()
   RigVMFunction_Sequence = &RigVMFunction_Sequence(ExecuteContext=@n1.ExecuteContext)
   &RigUnit_ParentConstraint(
    AdvancedSettings=(RotationOrderForFilter=XZY)
    Child=(Type=Bone,Name="dyn_chain_a_03_r")
    Parents=((Item=(Type=Bone,Name="dyn_chain_b_02_r"),Weight=0.492),(Item=(Type=Bone,Name="dyn_chain_a_02_r"),Weight=0.508))
    Weight=1
    ExecutePin=@RigVMFunction_Sequence.A
   )
   > &RigUnit_ParentConstraint(
    AdvancedSettings=(InterpolationType=Average,RotationOrderForFilter=XZY)
    Child=(Type=Bone,Name="base_front_dress_01_m")
    Parents=((Item=(Type=Bone,Name="base_front_dress_01_l"),Weight=0.5),(Item=(Type=Bone,Name="base_front_dress_01_r"),Weight=0.5))
    Weight=0
   )
   > &RigUnit_ParentConstraint(
    AdvancedSettings=(InterpolationType=Average,RotationOrderForFilter=XZY)
    Child=(Type=Bone,Name="base_back_dress_01_m")
    Parents=((Item=(Type=Bone,Name="base_sideBack_dress_01_r"),Weight=0.5),(Item=(Type=Bone,Name="base_sideBack_dress_01_l"),Weight=0.5))
    Weight=1
   )
   RigVMFunctionReturnNode(ExecuteContext=@RigVMFunction_Sequence.B)
   &RigUnit_ParentConstraint(
    AdvancedSettings=(RotationOrderForFilter=XZY)
    Child=(Type=Bone,Name="base_Necklace_m")
    ExecutePin=()
    Parents=((Item=(Type=Bone,Name="spine_04"),Weight=0.0238),(Item=(Type=Bone,Name="clavicle_l"),Weight=0.0263),(Item=(Type=Bone,Name="clavicle_r"),Weight=0.0288),(Item=(Type=Bone,Name="spine_05"),Weight=0.9212))
    Weight=1
   )
   &RigUnit_ParentConstraint(
    AdvancedSettings=(RotationOrderForFilter=XZY)
    Child=(Type=Bone,Name="dyn_Necklace_01_m")
    ExecutePin=()
    Parents=((Item=(Type=Bone,Name="spine_04"),Weight=0.0238),(Item=(Type=Bone,Name="clavicle_l"),Weight=0.0263),(Item=(Type=Bone,Name="clavicle_r"),Weight=0.0288),(Item=(Type=Bone,Name="spine_05"),Weight=0.9212))
    Weight=1
   )
   &RigUnit_ParentConstraint(
    AdvancedSettings=(RotationOrderForFilter=XZY)
    Child=(Type=Bone,Name="dyn_Necklace_02_m")
    ExecutePin=()
    Parents=((Item=(Type=Bone,Name="spine_04"),Weight=0.0238),(Item=(Type=Bone,Name="clavicle_l"),Weight=0.0267),(Item=(Type=Bone,Name="clavicle_r"),Weight=0.0289),(Item=(Type=Bone,Name="spine_05"),Weight=0.9207))
    Weight=1
   )
   &RigUnit_ParentConstraint(
    AdvancedSettings=(RotationOrderForFilter=XZY)
    Child=(Type=Bone,Name="dyn_Necklace_03_m")
    ExecutePin=()
    Parents=((Item=(Type=Bone,Name="clavicle_r"),Weight=0.0094),(Item=(Type=Bone,Name="spine_04"),Weight=0.0282),(Item=(Type=Bone,Name="clavicle_l"),Weight=0.0298),(Item=(Type=Bone,Name="spine_05"),Weight=0.9326))
    Weight=1
   )
  ]}
 New Function_ContainedGraph {nodes [
   n1 = RigVMFunctionEntryNode()
   RigVMFunction_Sequence = &RigVMFunction_Sequence(ExecuteContext=@n1.ExecuteContext)
   n2 = RigVMVariableNode() {variable=SPR L Leg Fwd}
   &RigUnit_ModifyTransforms(
    ItemToModify="((Item=(Type=Bone,Name=\"base_front_dress_01_l\"),Transform=(Rotation=(X=-0.0382,Y=0.0871,Z=0.0033,W=0.9955),Translation=(X=-0.4617,Y=0.2835,Z=3.5869),Scale3D=(X=1,Y=1,Z=1))),(Item=(Type=Bone,Name=\"base_front_dress_01_m\"),Transform=(Rotation=(X=-0.0755,Y=0.0 ...(truncated)"
    WeightMinimum=0
    ExecutePin=@RigVMFunction_Sequence.A
    Weight=@n2.Value
   )
   n3 = RigVMVariableNode() {variable=SPR R Leg Fwd}
   > &RigUnit_ModifyTransforms(
    ItemToModify="((Item=(Type=Bone,Name=\"base_front_dress_01_m\"),Transform=(Rotation=(X=0.0002,Y=-0.0278,Z=0.8026,W=0.5959),Translation=(X=-8.5694,Y=-4.7429,Z=-0.4952),Scale3D=(X=1,Y=1,Z=1))),(Item=(Type=Bone,Name=\"base_front_dress_01_r\"),Transform=(Rotation=(X=0.022,Y=0.2 ...(truncated)"
    WeightMinimum=0
    Weight=@n3.Value
   )
   > &RigUnit_ModifyTransforms(
    ItemToModify="((Item=(Type=Bone,Name=\"base_chain_a_01_r\"),Transform=(Rotation=(X=-0.1998,Y=0.1761,Z=-0.0736,W=0.9611),Translation=(X=0,Y=0,Z=0),Scale3D=(X=1,Y=1,Z=1))),(Item=(Type=Bone,Name=\"base_chain_b_r\"),Transform=(Rotation=(X=0.0067,Y=0.6428,Z=-0.0044,W=0.766),Tran ...(truncated)"
    WeightMinimum=0
    Weight=@n3.Value
   )
   Multiply = &RigVMFunction_MathDoubleMul(A=@n2.Value, B=@n3.Value)
   > &RigUnit_ModifyTransforms(
    ItemToModify="((Item=(Type=Bone,Name=\"base_front_dress_01_m\"),Transform=(Rotation=(X=0,Y=0,Z=-0.866,W=0.5),Translation=(X=5,Y=-6,Z=-1),Scale3D=(X=1,Y=1,Z=1))),(Item=(Type=Bone,Name=\"base_front_dress_01_l\"),Transform=(Rotation=(X=-0.0013,Y=0.0824,Z=-0.1141,W=0.99),Transl ...(truncated)"
    WeightMinimum=0
    Weight=@Multiply.Result
   )
   n4 = RigVMVariableNode() {variable=SPR L Leg Bkd}
   > &RigUnit_ModifyTransforms(
    ItemToModify="((Item=(Type=Bone,Name=\"base_back_dress_01_m\"),Transform=(Rotation=(X=0.0318,Y=0.3685,Z=0.0799,W=0.9256),Translation=(X=0,Y=0,Z=0),Scale3D=(X=1,Y=1,Z=1))),(Item=(Type=Bone,Name=\"base_front_dress_01_l\"),Transform=(Rotation=(X=0,Y=0,Z=0.8607,W=0.5091),Transl ...(truncated)"
    WeightMinimum=0
    Weight=@n4.Value
   )
   n5 = RigVMVariableNode() {variable=SPR R Leg Bkd}
   > &RigUnit_ModifyTransforms(
    ItemToModify="((Item=(Type=Bone,Name=\"base_back_dress_01_m\"),Transform=(Rotation=(X=0,Y=-0.3656,Z=0,W=-0.9308),Translation=(X=-2.24,Y=0,Z=-2.4113),Scale3D=(X=1,Y=1,Z=1))),(Item=(Type=Bone,Name=\"base_front_dress_01_m\"),Transform=(Rotation=(X=0.0173,Y=-0.0257,Z=0.5757,W=0 ...(truncated)"
    WeightMinimum=0
    Weight=@n5.Value
   )
   > &RigUnit_ModifyTransforms(
    ItemToModify="((Item=(Type=Bone,Name=\"base_back_dress_01_m\"),Transform=(Rotation=(X=-0.014,Y=-0.8161,Z=-0.0073,W=0.5776),Translation=(X=0,Y=0,Z=0),Scale3D=(X=1,Y=1,Z=1))),(Item=(Type=Bone,Name=\"base_sideBack_dress_01_r\"),Transform=(Rotation=(X=-0.0187,Y=0.0912,Z=0.1453, ...(truncated)"
    WeightMinimum=0
    Weight=&RigVMFunction_MathDoubleMul(A=@n4.Value, B=@n5.Value).Result
   )
   > &RigUnit_ModifyTransforms(
    ItemToModify="((Item=(Type=Bone,Name=\"base_pouch_l\"),Transform=(Rotation=(X=0.0872,Y=0,Z=0,W=0.9962),Translation=(X=-2,Y=0,Z=0),Scale3D=(X=1,Y=1,Z=1))),(Item=(Type=Bone,Name=\"base_side_dress_01_l\"),Transform=(Rotation=(X=0,Y=0.0215,Z=0,W=0.9998),Translation=(X=3.3519,Y= ...(truncated)"
    WeightMinimum=0
    Weight=RigVMVariableNode() {variable=SPR L Leg Otw}.Value
   )
   > &RigUnit_ModifyTransforms(
    ItemToModify="((Item=(Type=Bone,Name=\"base_dress_layer_top_c\"),Transform=(Rotation=(X=-0.1154,Y=-0.6857,Z=-0.1115,W=0.7099),Translation=(X=0,Y=0,Z=0),Scale3D=(X=1,Y=1,Z=1))),(Item=(Type=Bone,Name=\"base_dress_layer_top_d\"),Transform=(Rotation=(X=0.1087,Y=-0.6535,Z=0.0991 ...(truncated)"
    WeightMinimum=0
    Weight=RigVMVariableNode() {variable=SPR R Leg Otw}.Value
   )
   n6 = &RigVMFunction_MathDoubleRemap(SourceMaximum=1, SourceMinimum=0, TargetMaximum=0, TargetMinimum=1, bClamp=False, Value=RigVMVariableNode() {variable=Pub_IsCrouching}.Value)
   > &RigUnit_ModifyTransforms(
    ItemToModify="((Item=(Type=Bone,Name=\"dyn_sideBack_dress_03_r\"),Transform=(Rotation=(X=-0.037,Y=-0.8037,Z=-0.0249,W=-0.5934),Translation=(X=-9.9614,Y=0.0004,Z=-6.9509),Scale3D=(X=1,Y=1,Z=1))),(Item=(Type=Bone,Name=\"dyn_side_dress_03_r\"),Transform=(Rotation=(X=-0.0633,Y= ...(truncated)"
    WeightMinimum=0
    Weight=&RigVMFunction_MathDoubleMul(A=@n6.Result, B=RigVMVariableNode() {variable=SPR R Calf Bkd}.Value).Result
   )
   > &RigUnit_ModifyTransforms(
    ItemToModify="((Item=(Type=Bone,Name=\"dyn_sideBack_dress_03_r\"),Transform=(Rotation=(X=-0.0705,Y=0.0659,Z=-0.0076,W=0.9953),Translation=(X=0.8721,Y=12.2408,Z=4.6059),Scale3D=(X=1,Y=1,Z=1))),(Item=(Type=Bone,Name=\"dyn_side_dress_03_r\"),Transform=(Rotation=(X=-0.3565,Y=0. ...(truncated)"
    WeightMinimum=0
    Weight=&RigVMFunction_MathDoubleMul(A=@n6.Result, B=RigVMVariableNode() {variable=SPR R Calf Twist}.Value).Result
   )
   > &RigUnit_ModifyTransforms(
    ItemToModify="((Item=(Type=Bone,Name=\"base_front_dress_01_r\"),Transform=(Rotation=(X=0.0384,Y=-0.0883,Z=-0.005,W=0.9953),Translation=(X=0,Y=0,Z=0),Scale3D=(X=1,Y=1,Z=1))),(Item=(Type=Bone,Name=\"base_front_dress_01_m\"),Transform=(Rotation=(X=0,Y=0,Z=0.2316,W=0.9728),Tran ...(truncated)"
    WeightMinimum=0
    Weight=RigVMVariableNode() {variable=SPR R Leg Fwd_Extreme}.Value
   )
   > &RigUnit_ModifyTransforms(
    ItemToModify="((Item=(Type=Bone,Name=\"base_front_dress_01_r\"),Transform=(Rotation=(X=-0.0612,Y=0.1921,Z=0.1751,W=0.9637),Translation=(X=0,Y=0,Z=0),Scale3D=(X=1,Y=1,Z=1))),(Item=(Type=Bone,Name=\"base_side_dress_01_r\"),Transform=(Rotation=(X=0,Y=0.2331,Z=0,W=0.9725),Trans ...(truncated)"
    WeightMinimum=0
    Weight=RigVMVariableNode() {variable=SPR R Leg Itw}.Value
   )
   > &RigUnit_ModifyTransforms(
    ItemToModify="((Item=(Type=Bone,Name=\"base_sideBack_dress_01_l\"),Transform=(Rotation=(X=0,Y=0,Z=0.1565,W=-0.9877),Translation=(X=0,Y=0,Z=0),Scale3D=(X=1,Y=1,Z=1))),(Item=(Type=Bone,Name=\"base_side_dress_01_l\"),Transform=(Rotation=(X=0,Y=-0.1552,Z=0,W=0.9879),Translation ...(truncated)"
    WeightMinimum=0
    Weight=RigVMVariableNode() {variable=SPR L Leg Itw}.Value
   )
   AlphaInterp = &RigVMFunction_AlphaInterp(
    InterpSpeedDecreasing=5
    InterpSpeedIncreasing=5
    Scale=1
    bClampResult=False
    Value=&RigVMFunction_MathDoubleMul(A=RigVMVariableNode() {variable=PelvisProx}.Value, B=RigVMVariableNode() {variable=Pub_IsCrouching}.Value).Result
   )
   > &RigUnit_ModifyTransforms(
    ItemToModify=((Item=(Type=Bone,Name="base_front_dress_01_m"),Transform=(Rotation=(X=0,Y=0,Z=0.1736,W=0.9848),Translation=(X=0,Y=0,Z=0),Scale3D=(X=1,Y=1,Z=1))))
    WeightMinimum=0
    Weight=@AlphaInterp.Result
   )
   Multiply_5 = RigVMAggregateNode(
    B=RigVMVariableNode() {variable=Pub_IsCrouching}.Value
    A=RigVMVariableNode() {variable=PelvisProx}.Value
    C=RigVMVariableNode() {variable=Pub_IsMoving}.Value
   )
   n7 = &RigVMFunction_AlphaInterp(InterpSpeedDecreasing=5, InterpSpeedIncreasing=5, Scale=1, bClampResult=False, Value=@Multiply_5.Result)
   > &RigUnit_ModifyTransforms(
    ItemToModify=((Item=(Type=Bone,Name="base_front_dress_01_m"),Transform=(Rotation=(X=0,Y=0,Z=0.1736,W=0.9848),Translation=(X=0,Y=0,Z=0),Scale3D=(X=1,Y=1,Z=1))))
    WeightMinimum=0
    Weight=@n7.Result
   )
   > DeltaRotation v4 = RigVMFunctionReferenceNode(
    Rotation Order=ZYX
    child=(Type=Bone,Name="thigh_l")
    debugRotation=false
    outputMinMax=(X=0,Y=1)
    parent=(Type=Null,Name="thigh_l_null")
    premult=1
    rotationMinMax=(X=0,Y=20)
    triggerAxis=Y
    twistAxis=(X=1,Y=0,Z=0)
   )
   > DeltaRotation v4_1 = RigVMFunctionReferenceNode(
    Rotation Order=ZYX
    child=(Type=Bone,Name="thigh_l")
    debugRotation=false
    outputMinMax=(X=0,Y=1)
    parent=(Type=Null,Name="thigh_l_null")
    premult=1
    rotationMinMax=(X=0,Y=70)
    triggerAxis=Z
    twistAxis=(X=1,Y=0,Z=0)
   )
   > &RigUnit_ModifyTransforms(
    ItemToModify=((Item=(Type=Bone,Name="base_back_dress_01_m"),Transform=(Rotation=(X=-0.014,Y=-0.1032,Z=-0.1334,W=-0.9856),Translation=(X=0,Y=0,Z=0),Scale3D=(X=1,Y=1,Z=1))))
    WeightMinimum=0
    Weight=&RigVMFunction_MathDoubleMul(B=@"DeltaRotation v4".outputWeight, A=@"DeltaRotation v4_1".outputWeight).Result
   )
   RigVMFunctionReturnNode(ExecuteContext=@RigVMFunction_Sequence.B)
  ]}
 New Function_ContainedGraph {nodes [
   n1 = RigVMFunctionEntryNode()
   RigVMFunction_Sequence = &RigVMFunction_Sequence(ExecuteContext=@n1.ExecuteContext)
   n2 = &RigVMFunction_AlphaInterp(InterpSpeedDecreasing=5, InterpSpeedIncreasing=10, Scale=1, bClampResult=False, Value=RigVMVariableNode() {variable=Pub_IsCrouching}.Value)
   &RigUnit_ModifyTransforms(
    ItemToModify="((Item=(Type=Bone,Name=\"base_back_dress_01_m\"),Transform=(Rotation=(X=0,Y=0.0872,Z=0,W=0.9962),Translation=(X=0,Y=0,Z=-10),Scale3D=(X=1,Y=1,Z=1))),(Item=(Type=Bone,Name=\"base_sideBack_dress_01_r\"),Transform=(Rotation=(X=0.0151,Y=0.0858,Z=-0.173,W=0.9811),T ...(truncated)"
    WeightMinimum=0
    Weight=@n2.Result
    ExecutePin=@RigVMFunction_Sequence.A
   )
   AlphaInterp = &RigVMFunction_AlphaInterp(InterpSpeedDecreasing=5, InterpSpeedIncreasing=5, Scale=1, bClampResult=False, Value=RigVMVariableNode() {variable=Pub_IsKneeSliding}.Value)
   > &RigUnit_ModifyTransforms(
    ItemToModify="((Item=(Type=Bone,Name=\"base_back_dress_01_m\"),Transform=(Rotation=(X=0,Y=0.6428,Z=0,W=0.766),Translation=(X=-7,Y=0,Z=0),Scale3D=(X=1,Y=1,Z=1))),(Item=(Type=Bone,Name=\"base_sideBack_dress_01_r\"),Transform=(Rotation=(X=0,Y=0.4226,Z=0,W=0.9063),Translation=( ...(truncated)"
    WeightMinimum=0
    Weight=@AlphaInterp.Result
   )
   > &RigUnit_ModifyTransforms(
    ItemToModify="((Item=(Type=Bone,Name=\"base_dress_layer_mid_d\"),Transform=(Rotation=(X=0.0864,Y=-0.13,Z=0.0114,W=0.9877),Translation=(X=-5,Y=-5,Z=0),Scale3D=(X=1,Y=1,Z=1))),(Item=(Type=Bone,Name=\"base_dress_layer_mid_a\"),Transform=(Rotation=(X=0,Y=-0.4226,Z=0,W=0.9063),T ...(truncated)"
    WeightMinimum=0
    Weight=@AlphaInterp.Result
   )
   > &RigUnit_ModifyTransforms(
    ItemToModify="((Item=(Type=Bone,Name=\"base_front_dress_01_m\"),Transform=(Rotation=(X=0,Y=0,Z=0,W=1),Translation=(X=0,Y=0,Z=0),Scale3D=(X=1,Y=1,Z=1))),(Item=(Type=Bone,Name=\"base_front_dress_01_r\"),Transform=(Rotation=(X=0,Y=-0.1736,Z=0,W=0.9848),Translation=(X=0,Y=0,Z=2 ...(truncated)"
    WeightMinimum=0
    Weight=RigVMVariableNode() {variable=Pub_BaseballSliding}.Value
   )
   > DeltaRotation v4 = RigVMFunctionReferenceNode(
    Rotation Order=ZYX
    child=(Type=Bone,Name="calf_r")
    debugRotation=false
    outputMinMax=(X=0,Y=1)
    parent=(Type=Null,Name="calf_r_null")
    premult=1
    rotationMinMax=(X=0,Y=106)
    triggerAxis=Z
    twistAxis=(X=1,Y=0,Z=0)
   )
   Multiply = &RigVMFunction_MathDoubleMul(A=@"DeltaRotation v4".outputWeight, B=RigVMVariableNode() {variable=Pub_IsCrouching}.Value)
   > &RigUnit_ModifyTransforms(
    ItemToModify="((Item=(Type=Bone,Name=\"base_sideBack_dress_01_r\"),Transform=(Rotation=(X=-0.0298,Y=0.3407,Z=0.0819,W=0.9361),Translation=(X=0,Y=0,Z=0),Scale3D=(X=1,Y=1,Z=1))),(Item=(Type=Bone,Name=\"base_back_dress_01_m\"),Transform=(Rotation=(X=0,Y=0.342,Z=0,W=0.9397),Tra ...(truncated)"
    WeightMinimum=0
    Weight=@Multiply.Result
   )
   RigVMFunctionReturnNode(ExecuteContext=@RigVMFunction_Sequence.B)
   &RigUnit_ModifyTransforms(
    ExecutePin=()
    ItemToModify="((Item=(Type=Bone,Name=\"base_back_dress_01_m\"),Transform=(Rotation=(X=0,Y=0.0872,Z=0,W=0.9962),Translation=(X=0,Y=0,Z=-10),Scale3D=(X=1,Y=1,Z=1))),(Item=(Type=Bone,Name=\"base_sideBack_dress_01_r\"),Transform=(Rotation=(X=0.0151,Y=0.0858,Z=-0.173,W=0.9811),T ...(truncated)"
    Weight=1
    WeightMinimum=0
   )
  ]}
 New Function_ContainedGraph {nodes [
   n1 = RigVMFunctionEntryNode()
   RigVMFunction_Sequence = &RigVMFunction_Sequence(ExecuteContext=@n1.ExecuteContext)
   AlphaInterp = &RigVMFunction_AlphaInterp(InterpSpeedDecreasing=5, InterpSpeedIncreasing=5, Scale=1, bClampResult=False, Value=RigVMVariableNode() {variable=Pub_IsJumping}.Value)
   &RigUnit_ModifyTransforms(
    ItemToModify="((Item=(Type=Bone,Name=\"dyn_sideBack_dress_03_r\"),Transform=(Rotation=(X=-0.037,Y=-0.8037,Z=0.0249,W=0.5934),Translation=(X=9.9614,Y=0.0004,Z=6.9509),Scale3D=(X=1,Y=1,Z=1))),(Item=(Type=Bone,Name=\"dyn_side_dress_03_r\"),Transform=(Rotation=(X=-0.0481,Y=-0.0 ...(truncated)"
    WeightMinimum=0
    Weight=@AlphaInterp.Result
    ExecutePin=@RigVMFunction_Sequence.A
   )
   RigVMFunctionReturnNode(ExecuteContext=@RigVMFunction_Sequence.B)
   &RigUnit_ModifyTransforms(
    ExecutePin=()
    ItemToModify="((Item=(Type=Bone,Name=\"base_front_dress_01_m\"),Transform=(Rotation=(X=0,Y=0,Z=-0.5,W=0.866),Translation=(X=0,Y=0,Z=0),Scale3D=(X=1,Y=1,Z=1))),(Item=(Type=Bone,Name=\"base_back_dress_01_m\"),Transform=(Rotation=(X=0,Y=0.2588,Z=0,W=0.9659),Translation=(X=0,Y= ...(truncated)"
    Weight=1
    WeightMinimum=0
   )
  ]}
 New Function_ContainedGraph {nodes [
   n1 = RigVMFunctionEntryNode()
   RigVMFunction_Sequence = &RigVMFunction_Sequence(ExecuteContext=@n1.ExecuteContext)
   RigVMFunctionReturnNode(ExecuteContext=@RigVMFunction_Sequence.B)
   &RigUnit_ModifyTransforms(
    ExecutePin=()
    ItemToModify="((Item=(Type=Bone,Name=\"base_back_dress_01_m\"),Transform=(Rotation=(X=0,Y=-0.2588,Z=0,W=0.9659),Translation=(X=0,Y=0,Z=0),Scale3D=(X=1,Y=1,Z=1))),(Item=(Type=Bone,Name=\"base_sideBack_dress_01_l\"),Transform=(Rotation=(X=0,Y=-0.2164,Z=0,W=0.9763),Translation ...(truncated)"
    WeightMinimum=0
    Weight=RigVMVariableNode() {variable=Pub_Floating}.Value
   )
  ]}
 RigVMFunctionLibrary {nodes [
   Constraints = RigVMCollapseNode(ExecuteContext=())
   Constrollers = RigVMCollapseNode(Debug Controllers=False, ExecuteContext=())
   Crouching_Fix = RigVMCollapseNode(ExecuteContext=())
   DeltaRotation v4 = RigVMCollapseNode(
    ExecuteContext=()
    Rotation Order=ZYX
    child=(Type=Bone,Name="None")
    outputMinMax=(X=0,Y=1)
    parent=(Type=Bone,Name="None")
    premult=1
    rotationMinMax=(X=0,Y=0)
    twistAxis=(X=1,Y=0,Z=0)
   )
   Floating Fix = RigVMCollapseNode(ExecuteContext=())
   Jump_Fix = RigVMCollapseNode(ExecuteContext=())
   Skirt_fix = RigVMCollapseNode(ExecuteContext=())
   Variables = RigVMCollapseNode(ExecuteContext=())
  ]}
 RigVMModel {nodes [
   RigUnit_BeginExecution = &RigUnit_BeginExecution()
   Constrollers = RigVMFunctionReferenceNode(Debug Controllers=false, ExecuteContext=@RigUnit_BeginExecution.ExecutePin)
   > FN_DeformRig_FNC = RigVMFunctionReferenceNode(
    DeformRig_ Curve="(deformRig_l_upperarm_up_alpha_72_#1=False,deformRig_r_upperarm_up_alpha_73_#2=False,deformRig_l_upperarm_fr_alpha_74_#3=False,deformRig_r_upperarm_fr_alpha_75_CB51B8 ...(truncated)"
    DeformRig_Bool="(upperarm_1_#4=True,elbow_3_#5=True,wrist_7_#6=True,groin_8_#7=True,glute_17_#8=True,knee_9_40C408324CA261 ...(truncated)"
    DeformRig_Debug=(upperarm_debug_18_#4=false,wrist_debug_20_#6=False,lat_debug_26_#9=False,groinGlute_debug_27_#7=False)
    DeformRig_Value="(upperarm_up_18_#10=(Rotation=(X=0.0,Y=-0.0,Z=0.0,W=1.0),Translation=(X=0.0,Y=0.0,Z=8.0),Scale3D=(X=0.70,Y=1.0,Z=1.0)),upperarm_fwd_fr_19_#11=(Rotation=(X=0.0,Y=-0.0,Z=0.0,W=1.0),Translation=(X=0.0, ...(truncated)"
   )
   > Constraints = RigVMFunctionReferenceNode()
   > Variables = RigVMFunctionReferenceNode()
   > Skirt_fix = RigVMFunctionReferenceNode()
   > Crouching_Fix = RigVMFunctionReferenceNode()
   > Floating Fix = RigVMFunctionReferenceNode()
   > Jump_Fix = RigVMFunctionReferenceNode()
  ]}
]
hierarchy [
 root {loc=<0 0 0>, rot=<0 0 0 1>}
 pelvis {parent=root, loc=<0 -0.5301 98.0571>, rot=<0.0127 0.707 -0.0127 -0.707>}
 spine_01 {parent=pelvis, loc=<6.1208 -0.9328 0>, rot=<0 0 0.034 -0.9994>}
 spine_02 {parent=spine_01, loc=<6.5996 -0.2095 0>, rot=<0 0 0.0238 -0.9997>}
 spine_03 {parent=spine_02, loc=<6.4914 -0.1237 0>, rot=<0 0 0.0149 -0.9999>}
 spine_04 {parent=spine_03, loc=<6.9598 0.0794 0>, rot=<0 0 -0.0132 -0.9999>}
 spine_05 {parent=spine_04, loc=<6.9987 0 0>, rot=<0 0 -0.0015 -1>}
 neck_01 {parent=spine_05, loc=<11.3138 0.1529 0>, rot=<0 0 -0.2003 -0.9797>}
 neck_02 {parent=neck_01, loc=<4.5627 0 0>, rot=<0 0 0.0504 -0.9987>}
 head {parent=neck_02, loc=<5.3914 0 0>, rot=<0 0 0.115 -0.9934>}
 dyn_neckLeaf_A {parent=head, loc=<-2.851 4.1622 8.1201>, rot=<0.706 0.2479 0.6633 -0.0114>}
 dyn_neckLeaf_B {parent=head, loc=<-0.1116 2.4403 10.7214>, rot=<0.696 0.3465 0.6175 -0.1194>}
 dyn_neckLeaf_C {parent=head, loc=<2.4543 -2.9312 8.7783>, rot=<0.9082 0.0641 0.3517 -0.2177>}
 dyn_neckLeaf_D_base {parent=head, loc=<2.299 -9.2881 0.7093>, rot=<0.6702 -0.2256 0.0672 -0.7039>}
 dyn_neckLeaf_D_1 {parent=dyn_neckLeaf_D_base, loc=<4.834 3.9265 0.0679>, rot=<-0.1503 -0.1123 -0.2145 -0.9585>}
 dyn_neckLeaf_D_2 {parent=dyn_neckLeaf_D_base, loc=<4.4882 -4.3065 0>, rot=<0.026 -0.1598 0.0943 -0.9823>}
 dyn_neckLeaf_E {parent=head, loc=<2.7617 -5.9486 -7.9188>, rot=<0.2353 -0.3534 -0.2878 -0.8585>}
 dyn_neckLeaf_F {parent=head, loc=<-0.341 -0.484 -9.1817>, rot=<0.2285 -0.5811 -0.2858 -0.7269>}
 dyn_neckLeaf_G {parent=head, loc=<-2.578 2.3513 -7.6686>, rot=<0.0766 -0.6852 -0.1765 -0.7025>}
 base_C_hair_1 {parent=head, loc=<6.9517 -5.4626 0>, rot=<-0.0949 0.7007 0.7007 -0.0949>}
 dyn_C_hair_1 {parent=base_C_hair_1, loc=<0 0 0>, rot=<0 0 0 1>}
 dyn_C_hair_2 {parent=dyn_C_hair_1, loc=<10.6015 0 0>, rot=<0 -0.0172 0 -0.9999>}
 dyn_C_hair_3 {parent=dyn_C_hair_2, loc=<10.6008 0 0>, rot=<0 0.0789 0 -0.9969>}
 base_L_hair_1 {parent=head, loc=<6.9314 -3.0124 -7.9806>, rot=<0.0956 -0.7006 -0.7001 0.0996>}
 dyn_L_hair_1 {parent=base_L_hair_1, loc=<0 0 0>, rot=<0 0 0 1>}
 dyn_L_hair_2 {parent=dyn_L_hair_1, loc=<10.7733 0 0>, rot=<0.0014 -0.0775 0.0234 -0.9967>}
 dyn_L_hair_3 {parent=dyn_L_hair_2, loc=<10.7695 0 0>, rot=<-0.0067 0.1556 0.0018 -0.9878>}
 base_R_hair_1 {parent=head, loc=<6.9319 -3.0124 7.9806>, rot=<-0.0996 0.7001 0.7006 -0.0956>}
 dyn_R_hair_1 {parent=base_R_hair_1, loc=<0 0 0>, rot=<0 0 0 1>}
 dyn_R_hair_2 {parent=dyn_R_hair_1, loc=<10.7735 0 0.0001>, rot=<-0.0014 -0.0775 -0.0234 -0.9967>}
 dyn_R_hair_3 {parent=dyn_R_hair_2, loc=<10.7696 0 0>, rot=<0.0067 0.1556 -0.0018 -0.9878>}
 clavicle_l {parent=spine_05, loc=<5.4831 -0.1053 -3.2946>, rot=<0.0465 -0.7289 -0.009 -0.6829>}
 upperarm_l {parent=clavicle_l, loc=<9.8658 -1.2136 -0.4105>, rot=<0.0164 -0.3245 -0.0269 -0.9454>}
 lowerarm_l {parent=upperarm_l, loc=<26.465 -0.8345 0.259>, rot=<0.0036 0.0161 -0.2298 -0.9731>}
 hand_l {parent=lowerarm_l, loc=<22.7355 -0.3203 -0.6105>, rot=<-0.6784 -0.0223 0.0735 -0.7307>}
 hand_attach_l {parent=hand_l, loc=<8.5 -3 0>, rot=<0.7071 0.7071 0 0>}
 weapon_l {parent=hand_l, loc=<-0.1974 -3.4111 0.3811>, rot=<0 0 0 1>}
 thumb_01_l {parent=hand_l, loc=<1.3118 -0.3191 -2.459>, rot=<0.6546 -0.4042 0.0664 -0.6354>}
 thumb_02_l {parent=thumb_01_l, loc=<5.9336 -0.0448 0.0401>, rot=<-0.0037 0.0213 0.298 -0.9543>}
 thumb_03_l {parent=thumb_02_l, loc=<2.2223 0.3682 -0.0907>, rot=<0.0118 -0.013 0.0066 -0.9998>}
 index_metacarpal_l {parent=hand_l, loc=<1.6535 0.0183 -2.96>, rot=<0.0158 -0.0124 -0.0173 -0.9996>}
 index_01_l {parent=index_metacarpal_l, loc=<8.3008 0.0357 -0.0524>, rot=<-0.0044 0.0485 0.0924 -0.9945>}
 index_02_l {parent=index_01_l, loc=<3.3587 -0.3004 0.0452>, rot=<-0.002 0.0027 0.3225 -0.9466>}
 index_03_l {parent=index_02_l, loc=<2.5578 -0.123 -0.0252>, rot=<-0.002 0.0067 0.2558 -0.9667>}
 middle_metacarpal_l {parent=hand_l, loc=<2.1856 -0.1725 -0.5984>, rot=<0.006 0.0168 -0.0329 -0.9993>}
 middle_01_l {parent=middle_metacarpal_l, loc=<7.3275 -0.2447 0.0129>, rot=<-0.0208 0.0042 0.0858 -0.9961>}
 middle_02_l {parent=middle_01_l, loc=<4.0123 0.1401 0.1175>, rot=<-0.0218 0.0258 0.3512 -0.9357>}
 middle_03_l {parent=middle_02_l, loc=<2.6077 0.0534 0.1373>, rot=<0.0082 0.0188 0.1708 -0.9851>}
 ring_metacarpal_l {parent=hand_l, loc=<2.2074 -0.3625 1.3638>, rot=<-0.0284 0.0722 -0.0458 -0.9959>}
 ring_01_l {parent=ring_metacarpal_l, loc=<7.0295 -0.6197 -0.1446>, rot=<-0.0489 -0.0252 0.1892 -0.9804>}
 ring_02_l {parent=ring_01_l, loc=<3.6174 0.4647 -0.1027>, rot=<0 0 0.2806 -0.9598>}
 ring_03_l {parent=ring_02_l, loc=<2.4152 0.114 0.0199>, rot=<0 0 0.1189 -0.9929>}
 pinky_metacarpal_l {parent=hand_l, loc=<1.7616 -0.9753 3.3814>, rot=<-0.0397 0.0814 -0.069 -0.9935>}
 pinky_01_l {parent=pinky_metacarpal_l, loc=<7.1709 -0.4831 -0.1094>, rot=<-0.0654 -0.0035 0.2242 -0.9723>}
 pinky_02_l {parent=pinky_01_l, loc=<2.8274 0.0893 -0.1043>, rot=<0.0037 0.0011 0.2854 -0.9584>}
 pinky_03_l {parent=pinky_02_l, loc=<2.2046 0.0307 -0.0078>, rot=<-0.0002 0.0098 0.1445 -0.9895>}
 deform_wrist_dn_base_l {parent=hand_l, loc=<0 -1.5326 0>, rot=<0 0 0 -1>}
 deform_wrist_dn_l {parent=deform_wrist_dn_base_l, loc=<0 0 0>, rot=<0 0 0 1>}
 deform_wrist_up_base_l {parent=hand_l, loc=<0 2.3009 0>, rot=<0 0 0 -1>}
 deform_wrist_up_l {parent=deform_wrist_up_base_l, loc=<0 0 0>, rot=<0 0 0 1>}
 lowerarm_twist_02_l {parent=lowerarm_l, loc=<10.6707 0 0>, rot=<0 0 0 1>}
 lowerarm_twist_01_l {parent=lowerarm_l, loc=<18.316 0 0>, rot=<0 0 0 1>}
 deform_elbow_in_base_l {parent=lowerarm_l, loc=<0.1834 3.6062 -0.2572>, rot=<0 0 0 -1>}
 deform_elbow_in_l {parent=deform_elbow_in_base_l, loc=<0 0 0>, rot=<0 0 0 1>}
 deform_elbow_out_base_l {parent=lowerarm_l, loc=<0 -4 0>, rot=<0 0 0 1>}
 deform_elbow_out_l {parent=deform_elbow_out_base_l, loc=<0 0 0>, rot=<0 0 0 1>}
 upperarm_twist_01_l {parent=upperarm_l, loc=<7.2135 -0.1134 0>, rot=<0 0 0 1>}
 deform_upperarm_bk_base_l {parent=upperarm_twist_01_l, loc=<-6.4053 -4.3746 0.314>, rot=<0 0 0 -1>}
 deform_upperarm_bk_l {parent=deform_upperarm_bk_base_l, loc=<0 0 0>, rot=<0 0 0 1>}
 deform_upperarm_fr_base_l {parent=upperarm_twist_01_l, loc=<-5.6312 5.092 0.3354>, rot=<0 0 0 -1>}
 deform_upperarm_fr_l {parent=deform_upperarm_fr_base_l, loc=<0 0 0>, rot=<0 0 0 1>}
 deform_upperarm_up_base_l {parent=upperarm_twist_01_l, loc=<-8.891 0.4825 3.9517>, rot=<0 0 0 -1>}
 deform_upperarm_up_l {parent=deform_upperarm_up_base_l, loc=<0 0 0>, rot=<0 0 0 1>}
 upperarm_twist_02_l {parent=upperarm_l, loc=<19.7998 -0.1134 0>, rot=<0 0 0 1>}
 clavicle_r {parent=spine_05, loc=<5.483 -0.1053 3.2946>, rot=<0.7289 0.0465 -0.6829 0.009>}
 upperarm_r {parent=clavicle_r, loc=<-9.8658 1.2136 0.4105>, rot=<0.0164 -0.3245 -0.0269 -0.9454>}
 lowerarm_r {parent=upperarm_r, loc=<-26.465 0.8345 -0.259>, rot=<0.0036 0.0161 -0.2298 -0.9731>}
 hand_r {parent=lowerarm_r, loc=<-22.7355 0.3203 0.6106>, rot=<-0.6784 -0.0223 0.0735 -0.7307>}
 thumb_01_r {parent=hand_r, loc=<-1.3118 0.3191 2.459>, rot=<0.6546 -0.4042 0.0664 -0.6354>}
 thumb_02_r {parent=thumb_01_r, loc=<-5.9336 0.0448 -0.0401>, rot=<-0.0037 0.0213 0.298 -0.9543>}
 thumb_03_r {parent=thumb_02_r, loc=<-2.2223 -0.3682 0.0907>, rot=<0.0118 -0.013 0.0066 -0.9998>}
 index_metacarpal_r {parent=hand_r, loc=<-1.6535 -0.0183 2.96>, rot=<0.0158 -0.0124 -0.0173 -0.9996>}
 index_01_r {parent=index_metacarpal_r, loc=<-8.3008 -0.0357 0.0524>, rot=<-0.0044 0.0485 0.0924 -0.9945>}
 index_02_r {parent=index_01_r, loc=<-3.3587 0.3004 -0.0452>, rot=<-0.002 0.0027 0.3225 -0.9466>}
 index_03_r {parent=index_02_r, loc=<-2.5578 0.123 0.0252>, rot=<-0.002 0.0067 0.2558 -0.9667>}
 middle_metacarpal_r {parent=hand_r, loc=<-2.1856 0.1725 0.5984>, rot=<0.006 0.0168 -0.0329 -0.9993>}
 middle_01_r {parent=middle_metacarpal_r, loc=<-7.3275 0.2447 -0.0129>, rot=<-0.0208 0.0042 0.0858 -0.9961>}
 middle_02_r {parent=middle_01_r, loc=<-4.0123 -0.1401 -0.1175>, rot=<-0.0218 0.0258 0.3512 -0.9357>}
 middle_03_r {parent=middle_02_r, loc=<-2.6077 -0.0534 -0.1373>, rot=<0.0082 0.0188 0.1708 -0.9851>}
 ring_metacarpal_r {parent=hand_r, loc=<-2.2074 0.3625 -1.3638>, rot=<-0.0284 0.0722 -0.0458 -0.9959>}
 ring_01_r {parent=ring_metacarpal_r, loc=<-7.0295 0.6197 0.1446>, rot=<-0.0489 -0.0252 0.1892 -0.9804>}
 ring_02_r {parent=ring_01_r, loc=<-3.6174 -0.4647 0.1027>, rot=<0 0 0.2806 -0.9598>}
 ring_03_r {parent=ring_02_r, loc=<-2.4152 -0.114 -0.0199>, rot=<0 0 0.1189 -0.9929>}
 pinky_metacarpal_r {parent=hand_r, loc=<-1.7616 0.9753 -3.3814>, rot=<-0.0397 0.0814 -0.069 -0.9935>}
 pinky_01_r {parent=pinky_metacarpal_r, loc=<-7.1709 0.4831 0.1094>, rot=<-0.0654 -0.0035 0.2242 -0.9723>}
 pinky_02_r {parent=pinky_01_r, loc=<-2.8274 -0.0893 0.1043>, rot=<0.0037 0.0011 0.2854 -0.9584>}
 pinky_03_r {parent=pinky_02_r, loc=<-2.2046 -0.0307 0.0078>, rot=<-0.0002 0.0098 0.1445 -0.9895>}
 weapon_r {parent=hand_r, loc=<0.1974 3.4111 -0.3811>, rot=<0 0 0 1>}
 hand_attach_r {parent=hand_r, loc=<-8.5001 2.9999 0>, rot=<0 0 -0.7071 -0.7071>}
 deform_wrist_dn_base_r {parent=hand_r, loc=<-0.0001 1.5328 0>, rot=<0 0 0 -1>}
 deform_wrist_dn_r {parent=deform_wrist_dn_base_r, loc=<0 0 0>, rot=<0 0 0 1>}
 deform_wrist_up_base_r {parent=hand_r, loc=<-0.0002 -2.3007 -0.0001>, rot=<0 0 0 -1>}
 deform_wrist_up_r {parent=deform_wrist_up_base_r, loc=<0 0 0>, rot=<0 0 0 1>}
 lowerarm_twist_02_r {parent=lowerarm_r, loc=<-10.6707 0 0>, rot=<0 0 0 1>}
 lowerarm_twist_01_r {parent=lowerarm_r, loc=<-18.316 0 0>, rot=<0 0 0 1>}
 deform_elbow_in_base_r {parent=lowerarm_r, loc=<-0.1834 -3.6062 0.2572>, rot=<0 0 0 -1>}
 deform_elbow_in_r {parent=deform_elbow_in_base_r, loc=<0 0 0>, rot=<0 0 0 1>}
 deform_elbow_out_base_r {parent=lowerarm_r, loc=<0 4 -0.0001>, rot=<0 0 0 -1>}
 deform_elbow_out_r {parent=deform_elbow_out_base_r, loc=<0 0 0>, rot=<0 0 0 1>}
 upperarm_twist_01_r {parent=upperarm_r, loc=<-7.2135 0.1134 0>, rot=<0 0 0 1>}
 deform_upperarm_bk_base_r {parent=upperarm_twist_01_r, loc=<6.4053 4.3746 -0.314>, rot=<0 0 0 -1>}
 deform_upperarm_bk_r {parent=deform_upperarm_bk_base_r, loc=<0 0 0>, rot=<0 0 0 1>}
 deform_upperarm_fr_base_r {parent=upperarm_twist_01_r, loc=<5.6312 -5.092 -0.3354>, rot=<0 0 0 -1>}
 deform_upperarm_fr_r {parent=deform_upperarm_fr_base_r, loc=<0 0 0>, rot=<0 0 0 1>}
 deform_upperarm_up_base_r {parent=upperarm_twist_01_r, loc=<8.891 -0.4825 -3.9517>, rot=<0 0 0 -1>}
 deform_upperarm_up_r {parent=deform_upperarm_up_base_r, loc=<0 0 0>, rot=<0 0 0 1>}
 upperarm_twist_02_r {parent=upperarm_r, loc=<-19.7998 0.1134 0>, rot=<0 0 0 1>}
 attach_backpack {parent=spine_05, loc=<0 -8.35 0>, rot=<-0.0093 -0.707 0.0093 -0.707>}
 Backpack_BR {parent=attach_backpack, type=Null, loc=<0 0 0>, rot=<0 0 0 1>}
 attach_cape {parent=spine_05, loc=<9.3216 -8.4501 0>, rot=<0.0616 -0.7044 -0.0616 -0.7044>}
 cape {parent=attach_cape, type=Null, loc=<0 0 0>, rot=<0 0 0 1>}
 base_Necklace_m {parent=spine_05, loc=<10.0071 2.7475 0>, rot=<0 0 0.9457 0.325>}
 dyn_Necklace_01_m {parent=base_Necklace_m, loc=<0 0 0>, rot=<0 0 -0.0056 -1>}
 dyn_Necklace_02_m {parent=dyn_Necklace_01_m, loc=<3.8704 -0.0434 0>, rot=<0 0 0 1>}
 dyn_Necklace_03_m {parent=dyn_Necklace_02_m, loc=<3.0715 0 0>, rot=<0 0 -0.0405 -0.9992>}
 base_weapon_l {parent=spine_03, loc=<1.8977 1.4444 -11.9066>, rot=<-0.2773 -0.9562 0.0599 0.072>}
 dyn_weapon_l {parent=base_weapon_l, loc=<0 0 0>, rot=<0 0 0 1>}
 base_weapon_r {parent=spine_03, loc=<0.6515 1.3041 11.9761>, rot=<-0.2139 -0.9737 -0.0569 -0.0532>}
 dyn_weapon_r {parent=base_weapon_r, loc=<0 0 0>, rot=<0 0 0 1>}
 base_chain_b_r {parent=pelvis, loc=<9.8817 9.6528 2.0046>, rot=<-0.2423 -0.8552 0.216 0.404>}
 dyn_chain_b_01_r {parent=base_chain_b_r, loc=<0 0 0>, rot=<0 0 0 1>}
 dyn_chain_b_02_r {parent=dyn_chain_b_01_r, loc=<4.9909 -4.0678 0.5037>, rot=<0 0 0 1>}
 thigh_l {parent=pelvis, loc=<-1.0855 -0.2109 -8.7053>, rot=<0.0009 0.0355 0.0155 -0.9993>}
 calf_l {parent=thigh_l, loc=<-42.5284 -0.0019 -0.7297>, rot=<0.0006 -0.0161 -0.0391 -0.9991>}
 foot_l {parent=calf_l, loc=<-46.7402 -2.5014 -0.592>, rot=<0.0427 -0.0238 0.0353 -0.9982>}
 ball_l {parent=foot_l, loc=<-3.5658 10.8066 -0.1544>, rot=<0.0001 -0.0001 -0.7063 -0.7079>}
 calf_twist_02_l {parent=calf_l, loc=<-19.1546 -1.7943 -0.2638>, rot=<0.0237 0.0095 -0.0035 -0.9997>}
 calf_twist_01_l {parent=calf_l, loc=<-35.8483 -1.9197 -0.5794>, rot=<0.0237 0.0095 -0.0035 -0.9997>}
 deform_knee_in_base_l {parent=calf_l, loc=<0.8031 -5.0231 0.2456>, rot=<0 0 0 -1>}
 deform_knee_in_l {parent=deform_knee_in_base_l, loc=<0 0 0>, rot=<0 0 0 1>}
 deform_knee_out_base_l {parent=calf_l, loc=<0 4 0>, rot=<0 0 0 -1>}
 deform_knee_out_l {parent=deform_knee_out_base_l, loc=<0 0 0>, rot=<0 0 0 1>}
 thigh_twist_01_l {parent=thigh_l, loc=<-27.952 0 0>, rot=<0.0013 0 0 -1>}
 base_sideBack_dress_01_l {parent=thigh_twist_01_l, loc=<26.0739 -8.109 -1.3142>, rot=<0.0055 0.592 0.8043 -0.052>}
 dyn_sideBack_dress_01_l {parent=base_sideBack_dress_01_l, loc=<0 0 0>, rot=<0 0 0 1>}
 dyn_sideBack_dress_02_l {parent=dyn_sideBack_dress_01_l, loc=<10.0827 0 -0.0006>, rot=<0 -0.0018 0 -1>}
 dyn_sideBack_dress_03_l {parent=dyn_sideBack_dress_02_l, loc=<10.3441 0 0>, rot=<0 0 0 1>}
 base_side_dress_01_l {parent=thigh_twist_01_l, loc=<28.8899 -2.9237 -6.5692>, rot=<-0.0078 -0.999 0.041 -0.0178>}
 dyn_side_dress_01_l {parent=base_side_dress_01_l, loc=<0 0 0>, rot=<0 0 0 1>}
 dyn_side_dress_02_l {parent=dyn_side_dress_01_l, loc=<8.0363 0 0.0069>, rot=<0 -0.0018 0 -1>}
 dyn_side_dress_03_l {parent=dyn_side_dress_02_l, loc=<9.4593 0 -0.1647>, rot=<0 0 0 1>}
 base_front_dress_01_l {parent=thigh_twist_01_l, loc=<27.2878 9.3847 1.0677>, rot=<-0.081 -0.9966 0.003 0.014>}
 dyn_front_dress_01_l {parent=base_front_dress_01_l, loc=<0 0 0>, rot=<0 0 0 1>}
 dyn_front_dress_02_l {parent=dyn_front_dress_01_l, loc=<12.6815 -1.0742 0.2802>, rot=<0 0 0 1>}
 deform_groin_base_l {parent=thigh_l, loc=<0.4801 5.6776 -0.2486>, rot=<0 0 0 -1>}
 deform_groin_l {parent=deform_groin_base_l, loc=<0 0 0>, rot=<0 0 0 1>}
 thigh_r {parent=pelvis, loc=<-1.0855 -0.2109 8.7053>, rot=<0.0355 -0.0009 0.9993 0.0155>}
 calf_r {parent=thigh_r, loc=<42.5284 0.0019 0.7297>, rot=<0.0006 -0.0161 -0.0391 -0.9991>}
 foot_r {parent=calf_r, loc=<46.7401 2.5014 0.592>, rot=<0.0427 -0.0238 0.0353 -0.9982>}
 ball_r {parent=foot_r, loc=<3.5658 -10.8066 0.1544>, rot=<0.0001 -0.0001 -0.7063 -0.7079>}
 calf_twist_02_r {parent=calf_r, loc=<19.1546 1.7943 0.2638>, rot=<0.0237 0.0095 -0.0035 -0.9997>}
 calf_twist_01_r {parent=calf_r, loc=<35.8483 1.9197 0.5794>, rot=<0.0237 0.0095 -0.0035 -0.9997>}
 deform_knee_in_base_r {parent=calf_r, loc=<-0.8031 5.0231 -0.2457>, rot=<0 0 0 -1>}
 deform_knee_in_r {parent=deform_knee_in_base_r, loc=<0 0 0>, rot=<0 0 0 1>}
 deform_knee_out_base_r {parent=calf_r, loc=<0 -4 0>, rot=<0 0 0 -1>}
 deform_knee_out_r {parent=deform_knee_out_base_r, loc=<0 0 0>, rot=<0 0 0 1>}
 thigh_twist_01_r {parent=thigh_r, loc=<27.952 0 0>, rot=<0.0013 0 0 -1>}
 base_front_dress_01_r {parent=thigh_twist_01_r, loc=<-28.688 -7.8419 0.2775>, rot=<0.9964 -0.081 -0.023 0.0023>}
 dyn_front_dress_01_r {parent=base_front_dress_01_r, loc=<0 0 0>, rot=<0 0 0 1>}
 dyn_front_dress_02_r {parent=dyn_front_dress_01_r, loc=<24.2481 -2.7455 0.4338>, rot=<0 0 0 1>}
 base_side_dress_01_r {parent=thigh_twist_01_r, loc=<-27.6799 1.3209 7.4947>, rot=<0.9996 -0.0043 0.0212 0.0176>}
 dyn_side_dress_01_r {parent=base_side_dress_01_r, loc=<0 0 0>, rot=<0 0 0 1>}
 dyn_side_dress_02_r {parent=dyn_side_dress_01_r, loc=<25.0877 -0.0665 1.0349>, rot=<0 -0.0018 0 -1>}
 dyn_side_dress_03_r {parent=dyn_side_dress_02_r, loc=<16.7754 0 -0.1647>, rot=<0 0 0 1>}
 base_sideBack_dress_01_r {parent=thigh_twist_01_r, loc=<-22.7154 7.5494 -0.4643>, rot=<0.8703 0.063 0.0192 0.488>}
 dyn_sideBack_dress_01_r {parent=base_sideBack_dress_01_r, loc=<0 0 0>, rot=<0 0 0 1>}
 dyn_sideBack_dress_02_r {parent=dyn_sideBack_dress_01_r, loc=<12.1899 0 -0.0083>, rot=<0 -0.0018 0 -1>}
 dyn_sideBack_dress_03_r {parent=dyn_sideBack_dress_02_r, loc=<12.3906 0 -0.1647>, rot=<0 0 0 1>}
 base_dress_layer_mid_d {parent=thigh_twist_01_r, loc=<-23.118 5.4482 7.4938>, rot=<0.2339 0.1705 -0.051 -0.9558>}
 dyn_dress_layer_mid_d_01 {parent=base_dress_layer_mid_d, loc=<0 0 0>, rot=<0 0 0 1>}
 dyn_dress_layer_mid_d_02 {parent=dyn_dress_layer_mid_d_01, loc=<6.1151 -0.0742 -0.2324>, rot=<0 0 0 1>}
 base_dress_layer_mid_c {parent=thigh_twist_01_r, loc=<-28.8973 11.2895 1.3475>, rot=<0.5815 0.2031 -0.1719 -0.7688>}
 dyn_dress_layer_mid_c_01 {parent=base_dress_layer_mid_c, loc=<0 0 0>, rot=<0 0 0 1>}
 dyn_dress_layer_mid_c_02 {parent=dyn_dress_layer_mid_c_01, loc=<5.4977 -0.2895 -1.1096>, rot=<-0.0028 -0.0001 0.0128 -0.9999>}
 base_dress_layer_mid_e {parent=thigh_twist_01_r, loc=<-23.4999 -5.7516 5.1788>, rot=<-0.3948 0.0183 0.0245 -0.9182>}
 dyn_dress_layer_mid_e_01 {parent=base_dress_layer_mid_e, loc=<0 0 0>, rot=<0 0 0 1>}
 deform_groin_base_r {parent=thigh_r, loc=<-0.4801 -5.6776 0.2486>, rot=<0 0 0 -1>}
 deform_groin_r {parent=deform_groin_base_r, loc=<0 0 0>, rot=<0 0 0 1>}
 deform_glute_base_l {parent=pelvis, loc=<-7.3466 -9.0213 -9.1245>, rot=<0.0172 0.031 -0.4477 -0.8935>}
 deform_glute_l {parent=deform_glute_base_l, loc=<0 0 0>, rot=<0 0 0 1>}
 deform_glute_base_r {parent=pelvis, loc=<-7.3464 -9.0213 9.1245>, rot=<0.031 -0.0172 0.8935 -0.4477>}
 deform_glute_r {parent=deform_glute_base_r, loc=<0 0 0>, rot=<0 0 0 1>}
 base_pouch_l {parent=pelvis, loc=<3.35 -6.648 -14.2537>, rot=<0.0504 -0.9552 0.2326 -0.176>}
 dyn_pouch_l {parent=base_pouch_l, loc=<0 0 0>, rot=<0 0 0 1>}
 base_back_dress_01_m {parent=pelvis, loc=<-3.3483 -10.2464 0>, rot=<0.0163 -0.7069 -0.7069 0.0163>}
 dyn_back_dress_01_m {parent=base_back_dress_01_m, loc=<0 0 0>, rot=<0 0 0 1>}
 dyn_back_dress_02_m {parent=dyn_back_dress_01_m, loc=<9.9254 0 0>, rot=<0 -0.0018 0 -1>}
 dyn_back_dress_03_m {parent=dyn_back_dress_02_m, loc=<10.9483 0 -0.1647>, rot=<0 0 0 1>}
 base_front_dress_01_m {parent=pelvis, loc=<-0.8089 8.3992 0.6187>, rot=<-0.0984 -0.995 -0.0146 -0.0014>}
 dyn_front_dress_01_m {parent=base_front_dress_01_m, loc=<0 0 0>, rot=<0 0 0 1>}
 dyn_front_dress_02_m {parent=dyn_front_dress_01_m, loc=<9.9254 0 0>, rot=<0 -0.0018 0 -1>}
 dyn_front_dress_03_m {parent=dyn_front_dress_02_m, loc=<10.9483 0 -0.1647>, rot=<0 0 0 1>}
 thigh_l_null {parent=pelvis, type=Null, loc=<-1.0855 -0.2109 -8.7053>, rot=<0.0009 0.0355 0.0155 -0.9993>}
 thigh_r_null {parent=pelvis, type=Null, loc=<-1.0855 -0.2109 8.7053>, rot=<0.0355 -0.0009 0.9993 0.0155>}
 base_dress_layer_top_b {parent=pelvis, loc=<7.972 -7.6922 8.159>, rot=<0.2752 -0.4724 0.8326 -0.0889>}
 dyn_dress_layer_top_b_01 {parent=base_dress_layer_top_b, loc=<0 0 0>, rot=<0 0 0 1>}
 dyn_dress_layer_top_b_02 {parent=dyn_dress_layer_top_b_01, loc=<5.1411 -0.4138 0.0795>, rot=<-0.0183 -0.0005 0.0255 -0.9995>}
 base_dress_layer_top_c {parent=pelvis, loc=<7.9885 -4.7649 10.0831>, rot=<0.2 -0.2246 0.9532 -0.0315>}
 dyn_dress_layer_top_c_01 {parent=base_dress_layer_top_c, loc=<0 0 0>, rot=<0 0 0 1>}
 dyn_dress_layer_top_c_02 {parent=dyn_dress_layer_top_c_01, loc=<7.1336 0.1013 0.6875>, rot=<0 0 0 1>}
 base_dress_layer_top_d {parent=pelvis, loc=<7.7995 2.7536 10.9994>, rot=<0.183 0.0937 0.9785 0.0134>}
 dyn_dress_layer_top_d_01 {parent=base_dress_layer_top_d, loc=<0 0 0>, rot=<0 0 0 1>}
 dyn_dress_layer_top_d_02 {parent=dyn_dress_layer_top_d_01, loc=<6.0956 0.1013 0.6875>, rot=<0 0 0 1>}
 base_dress_layer_mid_a {parent=pelvis, loc=<4.788 -10.2326 -3.9364>, rot=<0.1727 -0.7546 0.5989 -0.2053>}
 dyn_dress_layer_mid_a {parent=base_dress_layer_mid_a, loc=<0 0 0>, rot=<0 0 0 1>}
 base_dress_layer_mid_b {parent=pelvis, loc=<2.4743 -11.7859 2.0287>, rot=<0.1697 -0.6828 0.6806 -0.2045>}
 dyn_dress_layer_mid_b_01 {parent=base_dress_layer_mid_b, loc=<0 0 0>, rot=<0 0 0 1>}
 base_dress_layer_top_a {parent=pelvis, loc=<7.8304 -10.0867 1.9842>, rot=<0.2068 -0.6101 0.7421 -0.185>}
 dyn_dress_layer_top_a_01 {parent=base_dress_layer_top_a, loc=<0 0 0>, rot=<0 0 0 1>}
 base_dress_layer_mid_f {parent=pelvis, loc=<8.0159 6.4611 8.6416>, rot=<0.0527 0.6122 0.7879 0.0412>}
 dyn_dress_layer_mid_f_01 {parent=base_dress_layer_mid_f, loc=<0 0 0>, rot=<0 0 0 1>}
 calf_r_null {parent=pelvis, type=Null, loc=<-43.4348 1.0985 12.4491>, rot=<-0.0193 0.0027 -0.9995 0.0235>}
 base_chain_a_01_r {parent=pelvis, loc=<8.8619 -6.0096 10.2593>, rot=<-0.2423 -0.8552 0.216 0.404>}
 dyn_chain_a_01_r {parent=base_chain_a_01_r, loc=<0 0 0>, rot=<0 0 0 -1>}
 dyn_chain_a_02_r {parent=dyn_chain_a_01_r, loc=<6.7287 0 0>, rot=<-0.0829 0.0172 -0.1209 -0.989>}
 dyn_chain_a_03_r {parent=dyn_chain_a_02_r, loc=<4.7721 0 0>, rot=<-0.2039 0.008 -0.4076 -0.89>}
 dyn_chain_a_04_r {parent=dyn_chain_a_03_r, loc=<3.5468 0 0>, rot=<-0.0153 -0.13 -0.1085 -0.9854>}
 dyn_chain_a_05_r {parent=dyn_chain_a_04_r, loc=<3.5222 0 0>, rot=<-0.0782 0 0 -0.9969>}
 base_chain_bullets {parent=dyn_chain_a_04_r, loc=<-0.0182 0.1445 -0.1988>, rot=<-0.2261 0.7174 0.237 -0.6149>}
 dyn_chain_bullets {parent=base_chain_bullets, loc=<0 0 0>, rot=<0 0 0 1>}
 spine_05_null {parent=pelvis, type=Null, loc=<32.9717 -4.2232 0>, rot=<0 0 0.058 -0.9983>}
 attach {parent=root, loc=<0 0 0>, rot=<0 0 0 1>}
 ik_hand_root {parent=root, loc=<0 0 0>, rot=<0 0 0 1>}
 ik_hand_gun {parent=ik_hand_root, loc=<-48.1743 3.3545 104.4381>, rot=<0.6276 0.1309 0.4183 -0.6434>}
 ik_hand_r {parent=ik_hand_gun, loc=<0 0 0>, rot=<0 0 0 1>}
 ik_hand_l {parent=ik_hand_gun, loc=<59.3271 67.6997 34.3524>, rot=<-0.6158 -0.7027 -0.3565 0>}
 ik_foot_root {parent=root, loc=<0 0 0>, rot=<0 0 0 1>}
 ik_foot_l {parent=ik_foot_root, loc=<14.8262 -7.3545 8.296>, rot=<0.0357 0.7045 0.0284 -0.7082>}
 ik_foot_r {parent=ik_foot_root, loc=<-14.8262 -7.3545 8.296>, rot=<0.7082 0.0284 -0.7045 0.0357>}
 root_fp {parent=root, loc=<0 0 0>, rot=<0 0 0 1>}
 attach_fp {parent=root_fp, loc=<0 0 0>, rot=<0 0 0 1>}
 camera_root {parent=root_fp, loc=<0 0 0>, rot=<0 0 0 1>}
 camera {parent=camera_root, loc=<0 0 0>, rot=<0 0 0 1>}
 attach_emote_01 {parent=root, loc=<0 0 0>, rot=<0 0 0 1>}
 attach_emote_02 {parent=root, loc=<0 0 0>, rot=<0 0 0 1>}
 attach_emote_03 {parent=root, loc=<0 0 0>, rot=<0 0 0 1>}
 warp_point_01 {parent=root, loc=<0 0 0>, rot=<0 0 0 1>}
 warp_point_02 {parent=root, loc=<0 0 0>, rot=<0 0 0 1>}
 warp_point_03 {parent=root, loc=<0 0 0>, rot=<0 0 0 1>}
 thigh_l_ctrl {controlType=EULER_TRANSFORM, type=Control, loc=<0 0 0>, rot=<0 0 0 1>}
 calf_l_ctrl {controlType=EULER_TRANSFORM, parent=thigh_l_ctrl, type=Control, loc=<0 0 0>, rot=<0 0 0 1>}
 thigh_r_ctrl {controlType=EULER_TRANSFORM, type=Control, loc=<0 0 0>, rot=<0 0 0 1>}
 calf_r_ctrl {controlType=EULER_TRANSFORM, parent=thigh_r_ctrl, type=Control, loc=<0 0 0>, rot=<0 0 0 1>}
]
variables [
 PelvisProx {type=double, default=0}
 Pub_BaseballSliding {type=double, default=0}
 Pub_DynamicsEnable {type=bool, default=False}
 Pub_Floating {type=double, default=0}
 Pub_IsCrouching {type=double, default=0}
 Pub_IsJumping {type=double, default=0}
 Pub_IsKneeSliding {type=double, default=0}
 Pub_IsMoving {type=double, default=0}
 Pub_Is_Parachute {type=double, default=0}
 Pub_IscrouchingMove {type=double, default=0}
 SPR L Leg Bkd {type=double, default=0}
 SPR L Leg Fwd {type=double, default=0}
 SPR L Leg Itw {type=double, default=0}
 SPR L Leg Otw {type=double, default=0}
 SPR R Calf Bkd {type=double, default=0}
 SPR R Calf Twist {type=double, default=0}
 SPR R Leg Bkd {type=double, default=0}
 SPR R Leg Fwd {type=double, default=0}
 SPR R Leg Fwd_Extreme {type=double, default=0}
 SPR R Leg Itw {type=double, default=0}
 SPR R Leg Otw {type=double, default=0}
]
