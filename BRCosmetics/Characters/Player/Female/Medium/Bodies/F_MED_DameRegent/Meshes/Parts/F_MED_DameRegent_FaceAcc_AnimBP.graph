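# Animation Blueprint for the Dame Regent face accessory that animates its hair bones.
# It layers body-mesh clavicle branches onto the copied parent pose, then runs a hair Control Rig fed by aim and gameplay states.
# Chained AnimDynamics springs follow: a looser pass, then a tighter-limited pass with additive scaling on hair tips.
# The EventGraph derives state variables from movement, skydiving, weapon, pickaxe, back bling, emote and front end.
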AnimBlueprint /BRCosmetics/Characters/Player/Female/Medium/Bodies/F_MED_DameRegent/Meshes/Parts/F_MED_DameRegent_FaceAcc_AnimBP
target_skeleton=&/F_MED_DameRegent_FaceAcc_Skeleton
graphs [
 AnimGraph {nodes [
   Output Pose(Result=Use cached pose 'CR_Cache_Blend'()) {layerGroup=None}
   n1 = Anim Dynamics - Bone: dyn_L_hair_2!(
    SimulationSpace:AnimPhysSimSpaceType Enum
    ComponentPose=Local To Component(LocalPose=Use cached pose 'CR_Cache_Input'())
    bAlphaBoolEnabled:bool=True
    GravityOverride:Vector="0,0,-980"
    ExternalForce="0,0,0"
   ) -> (Pose) {
    alphaInputType=Bool
    angularDampingOverride=1
    angularSpringConstant=20000
    bAngularSpring=true
    bOverrideAngularDamping=true
    bUseGravityOverride=true
    gravityScale=0.35
    simulationSpace=RootRelative
    alphaBoolBlend {blendOption=Linear}
    boundBone {boneName=dyn_L_hair_2}
    componentAppliedLinearAccClamp=<10000 10000 10000>
    componentLinearAccScale=<1 1 1>
    componentLinearVelScale=<1 1 1>
    gravityOverride {z=-980}
    physicsBodyDefinitions [{
      boundBone {boneName=dyn_L_hair_2}
      boxExtents=<30 50 50>
      constraintSetup {angularLimitsMax=<10 15 10>, angularLimitsMin=<-10 0 -10>, angularTarget=<1 0 0>}
      localJointOffset=<4 0 0>
     }]
   }
   n2 = Anim Dynamics - Bone: dyn_L_hair_3!(
    SimulationSpace:AnimPhysSimSpaceType Enum
    ComponentPose=@n1
    bAlphaBoolEnabled:bool=True
    GravityOverride:Vector="0,0,-980"
    ExternalForce="0,0,0"
   ) -> (Pose) {
    alphaInputType=Bool
    angularDampingOverride=1
    angularSpringConstant=20000
    bAngularSpring=true
    bOverrideAngularDamping=true
    bUseGravityOverride=true
    bUseSphericalLimits=true
    gravityScale=0.35
    simulationSpace=RootRelative
    alphaBoolBlend {blendOption=Linear}
    boundBone {boneName=dyn_L_hair_3}
    componentAppliedLinearAccClamp=<10000 10000 10000>
    componentLinearAccScale=<1 1 1>
    componentLinearVelScale=<1 1 1>
    gravityOverride {z=-980}
    physicsBodyDefinitions [{
      boundBone {boneName=dyn_L_hair_3}
      boxExtents=<20 30 30>
      constraintSetup {angularLimitsMax=<10 30 10>, angularLimitsMin=<-10 0 -10>, angularTarget=<1 0 0>}
      localJointOffset=<4 0 0>
     }]
   }
   n3 = Anim Dynamics - Bone: dyn_C_hair_2!(
    SimulationSpace:AnimPhysSimSpaceType Enum
    ComponentPose=@n2
    bAlphaBoolEnabled:bool=True
    GravityOverride:Vector="0,0,-980"
    ExternalForce="0,0,0"
   ) -> (Pose) {
    alphaInputType=Bool
    angularDampingOverride=1
    angularSpringConstant=20000
    bAngularSpring=true
    bOverrideAngularDamping=true
    bUseGravityOverride=true
    gravityScale=0.35
    simulationSpace=RootRelative
    alphaBoolBlend {blendOption=Linear}
    boundBone {boneName=dyn_C_hair_2}
    componentAppliedLinearAccClamp=<10000 10000 10000>
    componentLinearAccScale=<1 1 1>
    componentLinearVelScale=<1 1 1>
    gravityOverride {z=-980}
    physicsBodyDefinitions [{
      boundBone {boneName=dyn_C_hair_2}
      boxExtents=<30 50 50>
      constraintSetup {angularLimitsMax=<10 15 10>, angularLimitsMin=<-10 0 -10>, angularTarget=<1 0 0>}
      localJointOffset=<4 0 0>
     }]
   }
   n4 = Anim Dynamics - Bone: dyn_C_hair_3!(
    SimulationSpace:AnimPhysSimSpaceType Enum
    ComponentPose=@n3
    bAlphaBoolEnabled:bool=True
    GravityOverride:Vector="0,0,-980"
    ExternalForce="0,0,0"
   ) -> (Pose) {
    alphaInputType=Bool
    angularDampingOverride=1
    angularSpringConstant=20000
    bAngularSpring=true
    bOverrideAngularDamping=true
    bUseGravityOverride=true
    gravityScale=0.35
    simulationSpace=RootRelative
    alphaBoolBlend {blendOption=Linear}
    boundBone {boneName=dyn_C_hair_3}
    componentAppliedLinearAccClamp=<10000 10000 10000>
    componentLinearAccScale=<1 1 1>
    componentLinearVelScale=<1 1 1>
    gravityOverride {z=-980}
    physicsBodyDefinitions [{
      boundBone {boneName=dyn_C_hair_3}
      boxExtents=<20 30 30>
      constraintSetup {angularLimitsMax=<10 30 10>, angularLimitsMin=<-10 0 -10>, angularTarget=<1 0 0>}
      localJointOffset=<4 0 0>
     }]
   }
   n5 = Anim Dynamics - Bone: dyn_R_hair_2!(
    SimulationSpace:AnimPhysSimSpaceType Enum
    ComponentPose=@n4
    bAlphaBoolEnabled:bool=True
    GravityOverride:Vector="0,0,-980"
    ExternalForce="0,0,0"
   ) -> (Pose) {
    alphaInputType=Bool
    angularDampingOverride=1
    angularSpringConstant=20000
    bAngularSpring=true
    bOverrideAngularDamping=true
    bUseGravityOverride=true
    gravityScale=0.35
    simulationSpace=RootRelative
    alphaBoolBlend {blendOption=Linear}
    boundBone {boneName=dyn_R_hair_2}
    componentAppliedLinearAccClamp=<10000 10000 10000>
    componentLinearAccScale=<1 1 1>
    componentLinearVelScale=<1 1 1>
    gravityOverride {z=-980}
    physicsBodyDefinitions [{
      boundBone {boneName=dyn_R_hair_2}
      boxExtents=<30 50 50>
      constraintSetup {angularLimitsMax=<10 15 10>, angularLimitsMin=<-10 0 -10>, angularTarget=<1 0 0>}
      localJointOffset=<4 0 0>
     }]
   }
   n6 = Anim Dynamics - Bone: dyn_R_hair_3!(
    SimulationSpace:AnimPhysSimSpaceType Enum
    ComponentPose=@n5
    bAlphaBoolEnabled:bool=True
    GravityOverride:Vector="0,0,-980"
    ExternalForce="0,0,0"
   ) -> (Pose) {
    alphaInputType=Bool
    angularDampingOverride=1
    angularSpringConstant=20000
    bAngularSpring=true
    bOverrideAngularDamping=true
    bUseGravityOverride=true
    gravityScale=0.35
    simulationSpace=RootRelative
    alphaBoolBlend {blendOption=Linear}
    boundBone {boneName=dyn_R_hair_3}
    componentAppliedLinearAccClamp=<10000 10000 10000>
    componentLinearAccScale=<1 1 1>
    componentLinearVelScale=<1 1 1>
    gravityOverride {z=-980}
    physicsBodyDefinitions [{
      boundBone {boneName=dyn_R_hair_3}
      boxExtents=<20 30 30>
      constraintSetup {angularLimitsMax=<10 30 10>, angularLimitsMin=<-10 0 -10>, angularTarget=<1 0 0>}
      localJointOffset=<4 0 0>
     }]
   }
   n7 = Anim Dynamics - Bone: dyn_hair_front_01_r!(SimulationSpace:AnimPhysSimSpaceType Enum, ComponentPose=@n6, Alpha=1, GravityOverride:Vector="0,0,-980", ExternalForce="0,0,0") -> (Pose) {
    angularDampingOverride=1
    angularSpringConstant=20000
    bAngularSpring=true
    bOverrideAngularDamping=true
    bUseGravityOverride=true
    gravityScale=0.35
    simulationSpace=RootRelative
    alphaBoolBlend {blendOption=Linear}
    boundBone {boneName=dyn_hair_front_01_r}
    componentAppliedLinearAccClamp=<10000 10000 10000>
    componentLinearAccScale=<1 1 1>
    componentLinearVelScale=<1 1 1>
    gravityOverride {z=-980}
    physicsBodyDefinitions [{
      boundBone {boneName=dyn_hair_front_01_r}
      boxExtents=<30 40 40>
      constraintSetup {angularLimitsMax=<25 30 50>, angularLimitsMin=<-25 -30 -50>, angularTarget=<1 0 0>}
      localJointOffset=<4 0 0>
     }]
   }
   n8 = Anim Dynamics - Bone: dyn_hair_front_02_r!(SimulationSpace:AnimPhysSimSpaceType Enum, ComponentPose=@n7, Alpha=1, GravityOverride:Vector="0,0,-980", ExternalForce="0,0,0") -> (Pose) {
    angularDampingOverride=1
    angularSpringConstant=20000
    bAngularSpring=true
    bOverrideAngularDamping=true
    bUseGravityOverride=true
    gravityScale=0.35
    simulationSpace=RootRelative
    alphaBoolBlend {blendOption=Linear}
    boundBone {boneName=dyn_hair_front_02_r}
    componentAppliedLinearAccClamp=<10000 10000 10000>
    componentLinearAccScale=<1 1 1>
    componentLinearVelScale=<1 1 1>
    gravityOverride {z=-980}
    physicsBodyDefinitions [{
      boundBone {boneName=dyn_hair_front_02_r}
      boxExtents=<30 40 40>
      constraintSetup {angularLimitsMax=<25 30 50>, angularLimitsMin=<-25 -30 -50>, angularTarget=<1 0 0>}
      localJointOffset=<4 0 0>
     }]
   }
   n9 = Anim Dynamics - Bone: dyn_hair_front_03_r!(SimulationSpace:AnimPhysSimSpaceType Enum, ComponentPose=@n8, Alpha=1, GravityOverride:Vector="0,0,-980", ExternalForce="0,0,0") -> (Pose) {
    angularDampingOverride=1
    angularSpringConstant=20000
    bAngularSpring=true
    bOverrideAngularDamping=true
    bUseGravityOverride=true
    gravityScale=0.35
    simulationSpace=RootRelative
    alphaBoolBlend {blendOption=Linear}
    boundBone {boneName=dyn_hair_front_03_r}
    componentAppliedLinearAccClamp=<10000 10000 10000>
    componentLinearAccScale=<1 1 1>
    componentLinearVelScale=<1 1 1>
    gravityOverride {z=-980}
    physicsBodyDefinitions [{
      boundBone {boneName=dyn_hair_front_03_r}
      boxExtents=<30 40 40>
      constraintSetup {angularLimitsMax=<25 30 50>, angularLimitsMin=<-25 -30 -50>, angularTarget=<1 0 0>}
      localJointOffset=<4 0 0>
     }]
   }
   Save cached pose 'CR_Cache_Dynamics'(Pose=Component To Local(ComponentPose=@n9)) {cacheName=CR_Cache_Dynamics}
   n10 = Get Body_Mesh()
   n11 = Control Rig!(
    Source=Layered blend per bone(BasePose=Copy Pose From Mesh() {bCopyCurves=true, bUseAttachedParent=true}, BlendPoses_0=Copy Pose From Mesh(SourceMeshComponent=@n10)) {layerSetup [{branchFilters [{boneName=clavicle_l}, {boneName=clavicle_r}]}]}
    Alpha
    pub disableUpperLimbsCorrection:bool=Get disableUpperLimbsCorrection() -> (disableUpperLimbsCorrection:bool)
    pub diveFloating:bool=Get bIsSkydiveFloating()
    pub isTargeting:bool=Get bIsTargeting()
    pub aimPitch:double=Get AimPitch()
    pub isAssaultRifleEquipped:double=Get isAssaultRifleEquipped float() -> (isAssaultRifleEquipped float:double)
    Pub_PickaxeSwing:double=Get IsSwing() -> (IsSwing:double)
    YouBoreMeEmote:double=Get YouBoreMe() -> (YouBoreMe:double)
   ) -> (Pose) {
    alphaBoolBlend {blendOption=Linear}
    controlRigAssetReference {blueprintRigClass=&/F_MED_DameRegent_FaceAcc_Hair_CtrlRig.F_MED_DameRegent_FaceAcc_Hair_CtrlRig_C}
   }
   Save cached pose 'CR_Cache_Input'(Pose=@n11) {cacheName=CR_Cache_Input}
   n12 = Anim Dynamics - Bone: dyn_L_hair_2(
    ComponentPose=Local To Component(LocalPose=Use cached pose 'CR_Cache_Dynamics'())
    bAlphaBoolEnabled=True
    GravityOverride="0,0,-980"
    ExternalForce="0,0,0"
   ) {
    alphaInputType=Bool
    angularDampingOverride=1
    angularSpringConstant=20000
    bAngularSpring=true
    bOverrideAngularDamping=true
    bUseGravityOverride=true
    gravityScale=0.35
    simulationSpace=RootRelative
    alphaBoolBlend {blendOption=Linear}
    boundBone {boneName=dyn_L_hair_2}
    componentAppliedLinearAccClamp=<10000 10000 10000>
    componentLinearAccScale=<1 1 1>
    componentLinearVelScale=<1 1 1>
    gravityOverride {z=-980}
    physicsBodyDefinitions [{
      boundBone {boneName=dyn_L_hair_2}
      boxExtents=<30 50 50>
      constraintSetup {angularLimitsMax=<10 0 10>, angularLimitsMin=<-10 0 -10>, angularTarget=<1 0 0>}
      localJointOffset=<4 0 0>
     }]
   }
   n13 = Anim Dynamics - Bone: dyn_L_hair_3(ComponentPose=@n12, bAlphaBoolEnabled=True, GravityOverride="0,0,-980", ExternalForce="0,0,0") {
    alphaInputType=Bool
    angularDampingOverride=1
    angularSpringConstant=20000
    bAngularSpring=true
    bOverrideAngularDamping=true
    bUseGravityOverride=true
    gravityScale=0.35
    simulationSpace=RootRelative
    alphaBoolBlend {blendOption=Linear}
    boundBone {boneName=dyn_L_hair_3}
    componentAppliedLinearAccClamp=<10000 10000 10000>
    componentLinearAccScale=<1 1 1>
    componentLinearVelScale=<1 1 1>
    gravityOverride {z=-980}
    physicsBodyDefinitions [{
      boundBone {boneName=dyn_L_hair_3}
      boxExtents=<20 30 30>
      constraintSetup {angularLimitsMax=<10 0 10>, angularLimitsMin=<-10 0 -10>, angularTarget=<1 0 0>}
      localJointOffset=<4 0 0>
     }]
   }
   n14 = Anim Dynamics - Bone: dyn_C_hair_2(ComponentPose=@n13, bAlphaBoolEnabled=True, GravityOverride="0,0,-980", ExternalForce="0,0,0") {
    alphaInputType=Bool
    angularDampingOverride=1
    angularSpringConstant=20000
    bAngularSpring=true
    bOverrideAngularDamping=true
    bUseGravityOverride=true
    gravityScale=0.35
    simulationSpace=RootRelative
    alphaBoolBlend {blendOption=Linear}
    boundBone {boneName=dyn_C_hair_2}
    componentAppliedLinearAccClamp=<10000 10000 10000>
    componentLinearAccScale=<1 1 1>
    componentLinearVelScale=<1 1 1>
    gravityOverride {z=-980}
    physicsBodyDefinitions [{
      boundBone {boneName=dyn_C_hair_2}
      boxExtents=<30 50 50>
      constraintSetup {angularLimitsMax=<10 0 10>, angularLimitsMin=<-10 0 -10>, angularTarget=<1 0 0>}
      localJointOffset=<4 0 0>
     }]
   }
   n15 = Anim Dynamics - Bone: dyn_C_hair_3(ComponentPose=@n14, bAlphaBoolEnabled=True, GravityOverride="0,0,-980", ExternalForce="0,0,0") {
    alphaInputType=Bool
    angularDampingOverride=1
    angularSpringConstant=20000
    bAngularSpring=true
    bOverrideAngularDamping=true
    bUseGravityOverride=true
    gravityScale=0.35
    simulationSpace=RootRelative
    alphaBoolBlend {blendOption=Linear}
    boundBone {boneName=dyn_C_hair_3}
    componentAppliedLinearAccClamp=<10000 10000 10000>
    componentLinearAccScale=<1 1 1>
    componentLinearVelScale=<1 1 1>
    gravityOverride {z=-980}
    physicsBodyDefinitions [{
      boundBone {boneName=dyn_C_hair_3}
      boxExtents=<20 30 30>
      constraintSetup {angularLimitsMax=<10 0 10>, angularLimitsMin=<-10 0 -10>, angularTarget=<1 0 0>}
      localJointOffset=<4 0 0>
     }]
   }
   n16 = Anim Dynamics - Bone: dyn_R_hair_2(ComponentPose=@n15, bAlphaBoolEnabled=True, GravityOverride="0,0,-980", ExternalForce="0,0,0") {
    alphaInputType=Bool
    angularDampingOverride=1
    angularSpringConstant=20000
    bAngularSpring=true
    bOverrideAngularDamping=true
    bUseGravityOverride=true
    gravityScale=0.35
    simulationSpace=RootRelative
    alphaBoolBlend {blendOption=Linear}
    boundBone {boneName=dyn_R_hair_2}
    componentAppliedLinearAccClamp=<10000 10000 10000>
    componentLinearAccScale=<1 1 1>
    componentLinearVelScale=<1 1 1>
    gravityOverride {z=-980}
    physicsBodyDefinitions [{
      boundBone {boneName=dyn_R_hair_2}
      boxExtents=<30 50 50>
      constraintSetup {angularLimitsMax=<10 0 10>, angularLimitsMin=<-10 0 -10>, angularTarget=<1 0 0>}
      localJointOffset=<4 0 0>
     }]
   }
   n17 = Anim Dynamics - Bone: dyn_R_hair_3(ComponentPose=@n16, bAlphaBoolEnabled=True, GravityOverride="0,0,-980", ExternalForce="0,0,0") {
    alphaInputType=Bool
    angularDampingOverride=1
    angularSpringConstant=20000
    bAngularSpring=true
    bOverrideAngularDamping=true
    bUseGravityOverride=true
    gravityScale=0.35
    simulationSpace=RootRelative
    alphaBoolBlend {blendOption=Linear}
    boundBone {boneName=dyn_R_hair_3}
    componentAppliedLinearAccClamp=<10000 10000 10000>
    componentLinearAccScale=<1 1 1>
    componentLinearVelScale=<1 1 1>
    gravityOverride {z=-980}
    physicsBodyDefinitions [{
      boundBone {boneName=dyn_R_hair_3}
      boxExtents=<20 30 30>
      constraintSetup {angularLimitsMax=<10 0 10>, angularLimitsMin=<-10 0 -10>, angularTarget=<1 0 0>}
      localJointOffset=<4 0 0>
     }]
   }
   n18 = "Transform (Modify) Bone - Bone: dyn_L_hair_3"!(ComponentPose=@n17, bAlphaBoolEnabled:bool=True, Translation="0,0,0", Rotation="0,0,0", Scale="-0.075,-0.075,0.500") -> (Pose) {
    alphaInputType=Bool
    scaleMode=BMM_Additive
    alphaBoolBlend {blendInTime=1, blendOption=Linear, blendOutTime=1}
    boneToModify {boneName=dyn_L_hair_3}
    scale=<-0.075 -0.075 0.5>
   }
   n19 = "Transform (Modify) Bone - Bone: dyn_C_hair_3"!(ComponentPose=@n18, bAlphaBoolEnabled:bool=True, Translation="0,0,0", Rotation="0,0,0", Scale="1,1,0.5") -> (Pose) {
    alphaInputType=Bool
    scaleMode=BMM_Additive
    alphaBoolBlend {blendInTime=1, blendOption=Linear, blendOutTime=1}
    boneToModify {boneName=dyn_C_hair_3}
    scale {z=0.5}
   }
   n20 = "Transform (Modify) Bone - Bone: dyn_R_hair_3"!(ComponentPose=@n19, bAlphaBoolEnabled:bool=True, Translation="0,0,0", Rotation="0,0,0", Scale="1,1,0.5") -> (Pose) {
    alphaInputType=Bool
    scaleMode=BMM_Additive
    alphaBoolBlend {blendInTime=1, blendOption=Linear, blendOutTime=1}
    boneToModify {boneName=dyn_R_hair_3}
    scale {z=0.5}
   }
   Save cached pose 'CR_Cache_Blend'(Pose=Component To Local(ComponentPose=@n20)) {cacheName=CR_Cache_Blend}
  ]}
 EventGraph {nodes [
   n1 = Event BlueprintUpdateAnimation()
   > Set Delta Time X(Delta Time X:double=@n1.DeltaTimeX) -> (Output_Get:double)
   > n2 = Sequence!(execute) -> (then_0, then_1, then_2:Exec, then_3:Exec, then_4:Exec, then_5:Exec, then_6:Exec, then_7:Exec)
   > n3 = Gravity Override State Function DEPRECATED(CharacterPartAnimInstance=Self-Reference(), GameTimeInSeconds=GetGameTimeInSeconds(), GravityOverrideParams=Get Gravity Override Params())
   > Set GravityOverride(GravityOverride=@n3)
   n4 = Get bIsSkydiveDiveMode()
   n5 = Get bIsSkydiveFloating()
   n6 = Branch(execute=@n2.then_1, Condition=AND Boolean(A=Get isWearingPack() -> (isWearingPack:bool), B=OR Boolean(A=@n4, B=@n5)))
   > Set FixHairWithPack(FixHairWithPack:bool=true) -> (Output_Get:bool)
   Set FixHairWithPack(execute=@n6.else, FixHairWithPack=false)
   n7 = Branch(execute=@n2.then_2, Condition=Get isWearingPack())
   > Set toogleAnimDynamics(toogleAnimDynamics:bool=false) -> (Output_Get:bool)
   > Set EnabledynamicsPack(EnabledynamicsPack:bool=true) -> (Output_Get:bool)
   Set toogleAnimDynamics(execute=@n7.else, toogleAnimDynamics=true)
   > Set EnabledynamicsPack(EnabledynamicsPack=false)
   n8 = Get bIsJumping()
   n9 = Get bIsFalling()
   n10 = Get bIsSprinting()
   n11 = Branch(execute=@n2.then_3, Condition=AND Boolean(A=Get bIsTacticalSprinting(), B=OR Boolean!(A=@n9, B=@n8, C:bool=@n10) -> (ReturnValue)))
   > Set IsTacticalandJump(IsTacticalandJump:double=1) -> (Output_Get:double)
   Set IsTacticalandJump(execute=@n11.else, IsTacticalandJump=0)
   Set disableUpperLimbsCorrection(
    execute=@n2.then_4
    disableUpperLimbsCorrection:bool=OR Boolean(A=Get bIsSkydiveDiveMode(), B=Get bIsSkydiveFloating(), C=Get bIsParachuteOpened())
   ) -> (Output_Get:bool)
   > n12 = Branch(Condition=OR Boolean!(A=Get bIsSkydiveDiveMode(), B=Get bIsSkydiveFloating()) -> (ReturnValue))
   > Set IsSkyDive(IsSkyDive:double=1) -> (Output_Get:double)
   Set IsSkyDive(execute=@n12.else, IsSkyDive=0)
   n13 = Get bIsCrouchSprinting()
   n14 = Branch(execute=@n2.then_5, Condition=AND Boolean(A=Get bIsCrouching(), B=NOT Boolean(A=@n13)))
   > Set IsCrouchingNOTMove(IsCrouchingNOTMove:double=1) -> (Output_Get:double)
   Set IsCrouchingNOTMove(execute=@n14.else, IsCrouchingNOTMove=0)
   n15 = Sequence!(execute=@n2.then_6) -> (then_0, then_1, then_2)
   > n16 = Branch(Condition=Get bIsSkydiveFloating())
   > Do Once(Reset=@n17.then)
   > Set IsLauching(IsLauching:double=1) -> (Output_Get:double)
   > Delay(Duration=1)
   > Set IsLauching(IsLauching=0)
   Set IsLauching(execute=@n16.else, IsLauching=0)
   n17 = Branch(execute=@n15.then_1, Condition=NOT Boolean(A=Get bIsSkyDiving()))
   n18 = GetAnimInstance(self=GetOwnerSkeletalMeshComponent())
   Is Valid(exec=@n15.then_2, InputObject=@n18)
   > n19 = Set IsYouBoreMe(IsYouBoreMe:bool=Montage_IsPlaying(self=@n18, Montage=/BRCosmetics/Animation/Game/MainPlayer/Emotes/YouBoreMe/Emote_You_Bore_Me_CMF_M)) -> (Output_Get:bool)
   > n20 = Branch(Condition=@n19)
   > Set YouBoreMe(YouBoreMe:double=1) -> (Output_Get:double)
   Set YouBoreMe(execute=@n20.else, YouBoreMe=0)
   n21 = Branch(execute=@n2.then_7, Condition=Get bIsInFrontEnd())
   > Set SimlationSpace(SimlationSpace:AnimPhysSimSpaceType Enum=World) -> (Output_Get:AnimPhysSimSpaceType Enum)
   > Set AlphaFE(AlphaFE:double=0.5) -> (Output_Get:double)
   Set SimlationSpace(execute=@n21.else)
   > Set AlphaFE(AlphaFE=1)
   Event OnParachuteOpened()
   > Set IsParachuteOpened(IsParachuteOpened:double=1) -> (Output_Get:double)
   Event OnEndPickaxeSwing() {is=title K2Node_Event}
   > Set IsSwing(IsSwing:double=0) -> (Output_Get:double)
   Event OnBeginCrouching()
   > Set IsCrouching(IsCrouching:double=1) -> (Output_Get:double)
   Event OnEndCrouching()
   > Set IsCrouching(IsCrouching=0)
   n22 = Event OnCharacterCustomizationLoaded()
   > Set Body_Mesh(Body_Mesh=@n22.CustomBaseMesh)
   n23 = CheckForBackbling()
   > Set isWearingPack(isWearingPack:bool=OR Boolean!(A=@n23, B=@n23.bIsBackblingACape, C:bool=@n23.bIsBackblingLong) -> (ReturnValue)) -> (Output_Get:bool)
   > Set isWearingCape(isWearingCape=@n23.bIsBackblingLong)
   Event OnBeginSkydiving()
   > Set isDivingFloating float(isDivingFloating float:double=1) -> (Output_Get:double)
   Event OnEndSkydiving()
   > Set isDivingFloating float(isDivingFloating float=0)
   Event OnWeaponChanged()
   > Set isAssaultRifleEquipped float(isAssaultRifleEquipped float:double=0) -> (Output_Get:double)
   > Set IsOutlowShotgunEquipped(IsOutlowShotgunEquipped:double=0) -> (Output_Get:double)
   > Set IsSMGEquipped(IsSMGEquipped:double=0) -> (Output_Get:double)
   > n24 = Switch on EFortWeaponCoreAnimation!(execute, Selection:EFortWeaponCoreAnimation Enum=Get WeaponCoreAnim!() -> (WeaponCoreAnim)) -> (
    Melee:Exec
    Pistol:Exec
    Shotgun:Exec
    PaperBlueprint:Exec
    Rifle:Exec
    MeleeOneHand:Exec
    MachinePistol:Exec
    RocketLauncher:Exec
    GrenadeLauncher:Exec
    GoingCommando:Exec
    AssaultRifle:Exec
    TacticalShotgun:Exec
    SniperRifle:Exec
    TrapPlacement:Exec
    ShoulderLauncher:Exec
    AbilityDecoTool:Exec
    Crossbow:Exec
    C4:Exec
    RemoteControl:Exec
    DualWield:Exec
    AR_BullPup:Exec
    AR_ForwardGrip:Exec
    MedPackPaddles:Exec
    SMG_P90:Exec
    AR_DrumGun:Exec
    Consumable_Small:Exec
    Consumable_Large:Exec
    Balloon:Exec
    MountedTurret:Exec
    CreativeTool:Exec
    ExplosiveBow:Exec
    AshtonIndigo:Exec
    AshtonChicago:Exec
    MeleeDualWield:Exec
    Unarmed:Exec
   )
   Set isAssaultRifleEquipped float(execute=@n24.AssaultRifle, isAssaultRifleEquipped float=1)
   Set IsSMGEquipped(execute=@n24.SMG_P90, IsSMGEquipped=1)
   Set IsOutlowShotgunEquipped(execute=@n24.TacticalShotgun, IsOutlowShotgunEquipped=1)
   Event OnParachuteClosed()
   > Set IsParachuteOpened(IsParachuteOpened=0)
   Event OnBeginPickaxeSwing() {is=title K2Node_Event}
   > Set IsSwing(IsSwing=1)
   Event BlueprintInitializeAnimation()
   > Set Skeletal_Mesh(Skeletal_Mesh:Skeletal Mesh Component Object Reference=GetOwnerSkeletalMeshComponent()) -> (Output_Get:Skeletal Mesh Component Object Reference)
   n25 = Event BlueprintProcessAnimInputEvent()
   > n26 = Branch(Condition=IsNamed(self=@n25.AnimInputEvent, InEventName=aie.Player.Zipline.Entered))
   > Set isOnZipline(isOnZipline:bool=1) -> (Output_Get:bool)
   Branch(execute=@n26.else, Condition=IsNamed(self=@n25.AnimInputEvent, InEventName=aie.Player.Zipline.Exited))
   > Set isOnZipline(isOnZipline=false)
  ]}
]
variables [
 AlphaFE {type=number}
 alphaHairRigidBody float {type=number}
 angularSpringConstant {type=number}
 Body_Mesh {type=SkeletalMeshComponent}
 Delta Time X {type=number}
 disableUpperLimbsCorrection {type=boolean}
 EnabledynamicsPack {type=boolean}
 Fix HairFrontEnd Iddle {type=number}
 Fix HairFrontEnd IddleASW {type=boolean}
 FixHairWithPack {type=boolean}
 Gravity Override Params {type=GravityOverrideParamsStruct}
 GravityOverride {type=Vector}
 GravityOverride AnimDynamics {type=Vector}
 isAssaultRifleEquipped float {type=number}
 IsCrouching {type=number}
 IsCrouchingNOTMove {type=number}
 isDivingFloating float {type=number}
 IsLauching {type=number}
 isOnZipline {type=boolean}
 IsOutlowShotgunEquipped {type=number}
 IsParachuteOpened {type=number}
 IsSkyDive {type=number}
 isSlidling float {type=number}
 IsSMGEquipped {type=number}
 IsSwing {type=number}
 IsTacticalandJump {type=number}
 isTacticalJumping {type=boolean}
 isTacticalJumpingFalling {type=boolean}
 isUsingHangGlider float {type=boolean}
 isWearingCape {type=boolean}
 isWearingPack {type=boolean}
 IsYouBoreMe {type=boolean}
 SimlationSpace {type=AnimPhysSimSpaceType}
 Skeletal_Mesh {type=SkeletalMeshComponent}
 toogleAnimDynamics {type=boolean}
 YouBoreMe {type=number}
]
